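# Physics asset for the large female HipTripper player body cosmetic, covering its skirt bones.
# Capsules on pelvis, thighs and right calf twist sit alongside small base anchor capsules and chained capsules on back, front and side skirt bones.
# Constraints link most bases to the pelvis and dyn bones down each chain, with angular motion locked except the free left front_02 chain.
# An extra locked constraint ties the left side skirt tip to that free chain.
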PhysicsAsset /BRCosmetics/Characters/Player/Female/Large/Bodies/F_LRG_HipTripper/Meshes/F_LRG_HipTripper_Physics
bodies {
 base_necklace [{shapeType=Capsule, length=0.505, radius=0.505, shapeName=base_necklace_capsule, rotation=<rot 0 -90 0>}]
 base_skirt_back [{shapeType=Capsule, length=0.505, radius=0.505, shapeName=base_skirt_back_capsule, rotation=<rot 0 -90 0>}]
 base_skirt_front_01_l [{shapeType=Capsule, length=0.505, radius=0.505, shapeName=base_skirt_front_01_l_capsule, rotation=<rot 0 -90 0>}]
 base_skirt_front_02_01_l [{shapeType=Capsule, length=0.505, radius=0.505, shapeName=base_skirt_front_02_01_l_capsule, rotation=<rot 0 -90 0>}]
 base_skirt_front_02_01_r [{shapeType=Capsule, length=0.505, radius=0.505, shapeName=base_skirt_front_02_01_r_capsule, rotation=<rot 0 -90 0>}]
 base_skirt_side_01_l [{shapeType=Capsule, length=0.505, radius=0.505, shapeName=base_skirt_side_01_l_capsule, rotation=<rot 0 -90 0>}]
 base_skirt_side_01_r [{shapeType=Capsule, length=0.505, radius=0.505, shapeName=base_skirt_side_01_r_capsule, rotation=<rot 0 -90 0>}]
 calf_twist_01_r [{shapeType=Capsule, length=31.1477, radius=7.2096, shapeName=calf_twist_01_r_capsule, center=<-9 0 0>, rotation=<rot 0 -95 90>}]
 dyn_skirt_back_01 [{shapeType=Capsule, length=10, radius=5, shapeName=dyn_skirt_back_01_capsule, center=<5 0 0>, rotation=<rot 0 0 90>}]
 dyn_skirt_back_02 [{shapeType=Capsule, length=10, radius=5, shapeName=dyn_skirt_back_02_capsule, center=<5 0 4>, rotation=<rot 0 0 90>}]
 dyn_skirt_front_01_l [{shapeType=Capsule, length=5, radius=5, shapeName=dyn_skirt_front_01_l_capsule, center=<5 0 2>, rotation=<rot 90 90 90>}]
 dyn_skirt_front_02_01_l [{
   shapeType=Capsule
   length=10
   radius=7
   shapeName=dyn_skirt_front_02_01_l_capsule
   center=<9.0932 -0.0613 8.8>
   rotation=<rot -20.8695 95.4853 76.0192>
  }]
 dyn_skirt_front_02_01_r [{shapeType=Capsule, length=10, radius=5, shapeName=dyn_skirt_front_02_01_r_capsule, center=<5 0 3.5>, rotation=<rot 0 0 90>}]
 dyn_skirt_front_02_02_l [{
   shapeType=Capsule
   length=15
   radius=7
   shapeName=dyn_skirt_front_02_02_l_capsule
   center=<16.1401 0 2.2>
   rotation=<rot -2.4463 82.6528 76.5985>
  }]
 dyn_skirt_front_02_02_r [{shapeType=Capsule, length=10, radius=5, shapeName=dyn_skirt_front_02_02_r_capsule, center=<5 0 4>, rotation=<rot 0 0 90>}]
 dyn_skirt_front_02_03_r [{shapeType=Capsule, length=10, radius=5, shapeName=dyn_skirt_front_02_03_r_capsule, center=<5 0 4>, rotation=<rot 0 0 90>}]
 dyn_skirt_side_01_l [{shapeType=Capsule, length=15, radius=5, shapeName=dyn_skirt_side_01_l_capsule, center=<5 0 2.5>, rotation=<rot 0 0 90>}]
 dyn_skirt_side_01_r [{shapeType=Capsule, length=10, radius=5, shapeName=dyn_skirt_side_01_r_capsule, center=<5 0 0>, rotation=<rot 0 0 90>}]
 dyn_skirt_side_02_l [
  {shapeType=Capsule, length=15, radius=5, shapeName=dyn_skirt_side_02_l_capsule, center=<5 0 4>, rotation=<rot 0 0 90>}
  {
   shapeType=Capsule
   length=23.438
   radius=6.562
   shapeName=dyn_skirt_side_02_l_capsule
   center=<15.7735 -3.4953 2.1061>
   rotation=<rot -0.2604 -6.7831 87.8133>
  }
 ]
 dyn_skirt_side_02_r [{shapeType=Capsule, length=10, radius=5, shapeName=dyn_skirt_side_02_r_capsule, center=<5 0 4>, rotation=<rot 0 0 90>}]
 pelvis [{
   shapeType=Capsule
   length=15.7142
   radius=12.1711
   shapeName=pelvis_capsule
   center=<0.9831 -2.0235 0.4621>
   rotation=<rot -10.2257 -90.0001 2.9312>
  }]
 thigh_l [{shapeType=Capsule, length=30, radius=10, shapeName=thigh_l_capsule, center=<-18 -1.8765 -0.8107>, rotation=<rot -90 0 0>}]
 thigh_r [{shapeType=Capsule, length=29.282, radius=10.718, shapeName=thigh_r_capsule, center=<18 0 0>, rotation=<rot -90 0 0>}]
}
constraints [
 {
  bone1Name=dyn_skirt_back_01
  bone2Name=base_skirt_back
  swing1LimitDegrees=45
  swing1Motion=Locked
  swing2LimitDegrees=45
  swing2Motion=Locked
  twistLimitDegrees=45
  twistMotion=Locked
 }
 {
  bone1Name=dyn_skirt_back_02
  bone2Name=dyn_skirt_back_01
  swing1LimitDegrees=45
  swing1Motion=Locked
  swing2LimitDegrees=45
  swing2Motion=Locked
  twistLimitDegrees=45
  twistMotion=Locked
 }
 {
  bone1Name=dyn_skirt_front_02_01_r
  bone2Name=base_skirt_front_02_01_r
  swing1LimitDegrees=45
  swing1Motion=Locked
  swing2LimitDegrees=45
  swing2Motion=Locked
  twistLimitDegrees=45
  twistMotion=Locked
 }
 {
  bone1Name=dyn_skirt_front_02_02_r
  bone2Name=dyn_skirt_front_02_01_r
  swing1LimitDegrees=45
  swing1Motion=Locked
  swing2LimitDegrees=45
  swing2Motion=Locked
  twistLimitDegrees=45
  twistMotion=Locked
 }
 {
  bone1Name=dyn_skirt_front_02_03_r
  bone2Name=dyn_skirt_front_02_02_r
  swing1LimitDegrees=45
  swing1Motion=Locked
  swing2LimitDegrees=45
  swing2Motion=Locked
  twistLimitDegrees=45
  twistMotion=Locked
 }
 {
  bone1Name=dyn_skirt_side_01_l
  bone2Name=base_skirt_side_01_l
  swing1LimitDegrees=45
  swing1Motion=Locked
  swing2LimitDegrees=45
  swing2Motion=Locked
  twistLimitDegrees=45
  twistMotion=Locked
 }
 {
  bone1Name=dyn_skirt_side_02_l
  bone2Name=dyn_skirt_side_01_l
  swing1LimitDegrees=45
  swing1Motion=Locked
  swing2LimitDegrees=45
  swing2Motion=Locked
  twistLimitDegrees=45
  twistMotion=Locked
 }
 {
  bone1Name=dyn_skirt_side_01_r
  bone2Name=base_skirt_side_01_r
  swing1LimitDegrees=45
  swing1Motion=Locked
  swing2LimitDegrees=45
  swing2Motion=Locked
  twistLimitDegrees=45
  twistMotion=Locked
 }
 {
  bone1Name=dyn_skirt_side_02_r
  bone2Name=dyn_skirt_side_01_r
  swing1LimitDegrees=45
  swing1Motion=Locked
  swing2LimitDegrees=45
  swing2Motion=Locked
  twistLimitDegrees=45
  twistMotion=Locked
 }
 {
  bone1Name=base_necklace
  bone2Name=pelvis
  swing1LimitDegrees=45
  swing1Motion=Locked
  swing2LimitDegrees=45
  swing2Motion=Locked
  twistLimitDegrees=45
  twistMotion=Locked
 }
 {
  bone1Name=thigh_l
  bone2Name=pelvis
  swing1LimitDegrees=45
  swing1Motion=Locked
  swing2LimitDegrees=45
  swing2Motion=Locked
  twistLimitDegrees=45
  twistMotion=Locked
 }
 {
  bone1Name=thigh_r
  bone2Name=pelvis
  swing1LimitDegrees=45
  swing1Motion=Locked
  swing2LimitDegrees=45
  swing2Motion=Locked
  twistLimitDegrees=45
  twistMotion=Locked
 }
 {
  bone1Name=base_skirt_side_01_l
  bone2Name=pelvis
  swing1LimitDegrees=45
  swing1Motion=Locked
  swing2LimitDegrees=45
  swing2Motion=Locked
  twistLimitDegrees=45
  twistMotion=Locked
 }
 {
  bone1Name=base_skirt_back
  bone2Name=pelvis
  swing1LimitDegrees=45
  swing1Motion=Locked
  swing2LimitDegrees=45
  swing2Motion=Locked
  twistLimitDegrees=45
  twistMotion=Locked
 }
 {
  bone1Name=base_skirt_side_01_r
  bone2Name=pelvis
  swing1LimitDegrees=45
  swing1Motion=Locked
  swing2LimitDegrees=45
  swing2Motion=Locked
  twistLimitDegrees=45
  twistMotion=Locked
 }
 {
  bone1Name=base_skirt_front_02_01_r
  bone2Name=pelvis
  swing1LimitDegrees=45
  swing1Motion=Locked
  swing2LimitDegrees=45
  swing2Motion=Locked
  twistLimitDegrees=45
  twistMotion=Locked
 }
 {
  bone1Name=dyn_skirt_front_01_l
  bone2Name=base_skirt_front_01_l
  swing1LimitDegrees=45
  swing1Motion=Locked
  swing2LimitDegrees=45
  swing2Motion=Locked
  twistLimitDegrees=45
  twistMotion=Locked
 }
 {
  bone1Name=base_skirt_front_02_01_l
  bone2Name=pelvis
  swing1LimitDegrees=45
  swing1Motion=Locked
  swing2LimitDegrees=45
  swing2Motion=Locked
  twistLimitDegrees=45
  twistMotion=Locked
 }
 {
  bone1Name=dyn_skirt_front_02_01_l
  bone2Name=base_skirt_front_02_01_l
  swing1LimitDegrees=45
  swing1Motion=Free
  swing2LimitDegrees=45
  swing2Motion=Free
  twistLimitDegrees=5
  twistMotion=Free
 }
 {
  bone1Name=dyn_skirt_front_02_02_l
  bone2Name=dyn_skirt_front_02_01_l
  swing1LimitDegrees=45
  swing1Motion=Free
  swing2LimitDegrees=45
  swing2Motion=Free
  twistLimitDegrees=5
  twistMotion=Free
 }
 {
  bone1Name=calf_twist_01_r
  bone2Name=thigh_r
  swing1LimitDegrees=45
  swing1Motion=Locked
  swing2LimitDegrees=45
  swing2Motion=Locked
  twistLimitDegrees=45
  twistMotion=Locked
 }
 {
  bone1Name=dyn_skirt_side_02_l
  bone2Name=dyn_skirt_front_02_02_l
  swing1LimitDegrees=45
  swing1Motion=Locked
  swing2LimitDegrees=45
  swing2Motion=Locked
  twistLimitDegrees=45
  twistMotion=Locked
 }
]
