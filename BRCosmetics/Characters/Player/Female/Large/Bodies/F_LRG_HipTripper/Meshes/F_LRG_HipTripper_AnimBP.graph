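# Animation Blueprint for the F_LRG_HipTripper character body, adding secondary motion to skirt, necklace and bangles.
# The parent mesh pose feeds a crouch-aware Control Rig, RigidBody physics, chained per-bone AnimDynamics and additive bone offsets.
# A second RigidBody asset and corrective offsets apply only in the front end.
# The update event computes state-driven gravity for the first RigidBody, plus simulation space, jump and directional crouch values.
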AnimBlueprint /BRCosmetics/Characters/Player/Female/Large/Bodies/F_LRG_HipTripper/Meshes/F_LRG_HipTripper_AnimBP
target_skeleton=&/F_LRG_HipTripper_Skeleton
graphs [
 AnimGraph {nodes [
   Anim Dynamics - Bone: dyn_necklace_01(Alpha=1, ExternalForce="0,0,0") {
    angularDampingOverride=1
    angularSpringConstant=4000
    bAngularSpring=true
    bOverrideAngularDamping=true
    bUseGravityOverride=true
    simulationSpace=RootRelative
    alphaBoolBlend {blendOption=Linear}
    boundBone {boneName=dyn_necklace_01}
    componentAppliedLinearAccClamp=<4000 4000 4000>
    componentLinearAccScale=<0.2 0.2 0.2>
    componentLinearVelScale=<0.2 0.2 0.2>
    gravityOverride {z=-980}
    physicsBodyDefinitions [{
      boundBone {boneName=dyn_necklace_01}
      boxExtents=<20 15 15>
      constraintSetup {angularLimitsMax=<0 0 5>, angularLimitsMin=<0 0 -5>, angularTarget=<1 0 0>}
      localJointOffset=<2 0 0>
     }]
   }
   n1 = Control Rig!(Source=Copy Pose From Mesh() {bUseAttachedParent=true}, Alpha, crouch_notmove:bool=Get bIsCrouching()) -> (Pose) {
    alphaBoolBlend {blendOption=Linear}
    controlRigAssetReference {blueprintRigClass=&/F_LRG_HipTripper_CtrlRig.F_LRG_HipTripper_CtrlRig_C}
   }
   n2 = RigidBody!(ComponentPose=Local To Component(LocalPose=@n1), Alpha, OverrideWorldGravity:Vector=Get Gravity Override(), ExternalForce) -> (Pose) {
    bClampLinearTranslationLimitToRefPose=true
    bOverrideWorldGravity=true
    overridePhysicsAsset=&/F_LRG_HipTripper_Physics
    simulationSpace=BaseBoneSpace
    alphaBoolBlend {blendOption=Linear}
    alphaScaleBiasClamp {interpSpeedDecreasing=15, interpSpeedIncreasing=20}
    baseBoneRef {boneName=pelvis}
    overrideWorldGravity {z=-980}
    simSpaceSettings {maxAngularAcceleration=500, maxAngularVelocity=4, maxLinearAcceleration=700, velocityScaleZ=0, worldAlpha=0.5}
   }
   n3 = Anim Dynamics - Bone: dyn_skirt_front_02_01_l!(SimulationSpace:AnimPhysSimSpaceType Enum, ComponentPose=@n2, Alpha=1, GravityOverride:Vector="0,0,-980", ExternalForce="0,0,0") -> (Pose) {
    angularDampingOverride=0.9999
    angularSpringConstant=400
    bAngularSpring=true
    bOverrideAngularDamping=true
    bUseGravityOverride=true
    gravityScale=0.35
    simulationSpace=RootRelative
    alphaBoolBlend {blendOption=Linear}
    boundBone {boneName=dyn_skirt_front_02_01_l}
    componentAppliedLinearAccClamp=<2000 2000 2000>
    componentLinearAccScale=<1 1 1>
    componentLinearVelScale=<1 1 1>
    gravityOverride {z=-980}
    physicsBodyDefinitions [{
      boundBone {boneName=dyn_skirt_front_02_01_l}
      boxExtents=<30 20 20>
      constraintSetup {angularLimitsMax=<0 0 0>, angularLimitsMin=<0 -25 -5>, angularTarget=<1 0 0>}
      localJointOffset=<3 0 0>
     }]
   }
   n4 = Anim Dynamics - Bone: dyn_skirt_front_02_02_l!(SimulationSpace:AnimPhysSimSpaceType Enum, ComponentPose=@n3, Alpha=1, GravityOverride:Vector="0,0,-980", ExternalForce="0,0,0") -> (Pose) {
    angularDampingOverride=0.9999
    angularSpringConstant=400
    bAngularSpring=true
    bOverrideAngularDamping=true
    bUseGravityOverride=true
    gravityScale=0.35
    simulationSpace=RootRelative
    alphaBoolBlend {blendOption=Linear}
    boundBone {boneName=dyn_skirt_front_02_02_l}
    componentAppliedLinearAccClamp=<2000 2000 2000>
    componentLinearAccScale=<1 1 1>
    componentLinearVelScale=<1 1 1>
    gravityOverride {z=-980}
    physicsBodyDefinitions [{
      boundBone {boneName=dyn_skirt_front_02_02_l}
      boxExtents=<30 20 20>
      constraintSetup {angularLimitsMax=<0 0 0>, angularLimitsMin=<0 -25 -5>, angularTarget=<1 0 0>}
      localJointOffset=<3 0 0>
     }]
   }
   n5 = Anim Dynamics - Bone: dyn_skirt_front_02_03_l!(SimulationSpace:AnimPhysSimSpaceType Enum, ComponentPose=@n4, Alpha=1, GravityOverride:Vector="0,0,-980", ExternalForce="0,0,0") -> (Pose) {
    angularDampingOverride=0.9999
    angularSpringConstant=400
    bAngularSpring=true
    bOverrideAngularDamping=true
    bUseGravityOverride=true
    gravityScale=0.35
    simulationSpace=RootRelative
    alphaBoolBlend {blendOption=Linear}
    boundBone {boneName=dyn_skirt_front_02_03_l}
    componentAppliedLinearAccClamp=<2000 2000 2000>
    componentLinearAccScale=<1 1 1>
    componentLinearVelScale=<1 1 1>
    gravityOverride {z=-980}
    physicsBodyDefinitions [{
      boundBone {boneName=dyn_skirt_front_02_03_l}
      boxExtents=<30 20 20>
      constraintSetup {angularLimitsMax=<0 0 0>, angularLimitsMin=<0 -25 -5>, angularTarget=<1 0 0>}
      localJointOffset=<3 0 0>
     }]
   }
   n6 = Anim Dynamics - Bone: dyn_skirt_side_01_l!(SimulationSpace:AnimPhysSimSpaceType Enum, ComponentPose=@n5, Alpha=1, GravityOverride:Vector="0,0,-980", ExternalForce="0,0,0") -> (Pose) {
    angularDampingOverride=0.9999
    angularSpringConstant=400
    bAngularSpring=true
    bOverrideAngularDamping=true
    bUseGravityOverride=true
    gravityScale=0.35
    simulationSpace=RootRelative
    alphaBoolBlend {blendOption=Linear}
    boundBone {boneName=dyn_skirt_side_01_l}
    componentAppliedLinearAccClamp=<2000 2000 2000>
    componentLinearAccScale=<1 1 1>
    componentLinearVelScale=<1 1 1>
    gravityOverride {z=-980}
    physicsBodyDefinitions [{
      boundBone {boneName=dyn_skirt_side_01_l}
      boxExtents=<30 20 20>
      constraintSetup {angularLimitsMax=<0 0 0>, angularLimitsMin=<0 -25 -5>, angularTarget=<1 0 0>}
      localJointOffset=<3 0 0>
     }]
   }
   n7 = Anim Dynamics - Bone: dyn_skirt_side_02_l!(SimulationSpace:AnimPhysSimSpaceType Enum, ComponentPose=@n6, Alpha=1, GravityOverride:Vector="0,0,-980", ExternalForce="0,0,0") -> (Pose) {
    angularDampingOverride=0.9999
    angularSpringConstant=400
    bAngularSpring=true
    bOverrideAngularDamping=true
    bUseGravityOverride=true
    gravityScale=0.35
    simulationSpace=RootRelative
    alphaBoolBlend {blendOption=Linear}
    boundBone {boneName=dyn_skirt_side_02_l}
    componentAppliedLinearAccClamp=<2000 2000 2000>
    componentLinearAccScale=<1 1 1>
    componentLinearVelScale=<1 1 1>
    gravityOverride {z=-980}
    physicsBodyDefinitions [{
      boundBone {boneName=dyn_skirt_side_02_l}
      boxExtents=<30 20 20>
      constraintSetup {angularLimitsMax=<0 0 0>, angularLimitsMin=<0 -25 -5>, angularTarget=<1 0 0>}
      localJointOffset=<3 0 0>
     }]
   }
   n8 = Anim Dynamics - Bone: dyn_skirt_back_01!(SimulationSpace:AnimPhysSimSpaceType Enum, ComponentPose=@n7, Alpha=0.7, GravityOverride:Vector="0,0,-980", ExternalForce="0,0,0") -> (Pose) {
    angularDampingOverride=0.9999
    angularSpringConstant=400
    bAngularSpring=true
    bOverrideAngularDamping=true
    bUseGravityOverride=true
    gravityScale=0.35
    simulationSpace=RootRelative
    alphaBoolBlend {blendOption=Linear}
    boundBone {boneName=dyn_skirt_back_01}
    componentAppliedLinearAccClamp=<2000 2000 2000>
    componentLinearAccScale=<1 1 1>
    componentLinearVelScale=<1 1 1>
    gravityOverride {z=-980}
    physicsBodyDefinitions [{
      boundBone {boneName=dyn_skirt_back_01}
      boxExtents=<30 20 20>
      constraintSetup {angularLimitsMax=<0 0 0>, angularLimitsMin=<0 -25 -5>, angularTarget=<1 0 0>}
      localJointOffset=<3 0 0>
     }]
   }
   n9 = Anim Dynamics - Bone: dyn_skirt_back_02!(SimulationSpace:AnimPhysSimSpaceType Enum, ComponentPose=@n8, Alpha=1, GravityOverride:Vector="0,0,-980", ExternalForce="0,0,0") -> (Pose) {
    angularDampingOverride=0.9999
    angularSpringConstant=400
    bAngularSpring=true
    bOverrideAngularDamping=true
    bUseGravityOverride=true
    gravityScale=0.35
    simulationSpace=RootRelative
    alphaBoolBlend {blendOption=Linear}
    boundBone {boneName=dyn_skirt_back_02}
    componentAppliedLinearAccClamp=<2000 2000 2000>
    componentLinearAccScale=<1 1 1>
    componentLinearVelScale=<1 1 1>
    gravityOverride {z=-980}
    physicsBodyDefinitions [{
      boundBone {boneName=dyn_skirt_back_02}
      boxExtents=<30 20 20>
      constraintSetup {angularLimitsMax=<0 0 0>, angularLimitsMin=<0 -25 -5>, angularTarget=<1 0 0>}
      localJointOffset=<3 0 0>
     }]
   }
   n10 = Anim Dynamics - Bone: dyn_skirt_side_01_r!(SimulationSpace:AnimPhysSimSpaceType Enum, ComponentPose=@n9, Alpha=1, GravityOverride:Vector="0,0,-980", ExternalForce="0,0,0") -> (Pose) {
    angularDampingOverride=0.9999
    angularSpringConstant=400
    bAngularSpring=true
    bOverrideAngularDamping=true
    bUseGravityOverride=true
    gravityScale=0.35
    simulationSpace=RootRelative
    alphaBoolBlend {blendOption=Linear}
    boundBone {boneName=dyn_skirt_side_01_r}
    componentAppliedLinearAccClamp=<2000 2000 2000>
    componentLinearAccScale=<1 1 1>
    componentLinearVelScale=<1 1 1>
    gravityOverride {z=-980}
    physicsBodyDefinitions [{
      boundBone {boneName=dyn_skirt_side_01_r}
      boxExtents=<30 20 20>
      constraintSetup {angularLimitsMax=<0 0 0>, angularLimitsMin=<0 -25 -5>, angularTarget=<1 0 0>}
      localJointOffset=<3 0 0>
     }]
   }
   n11 = Anim Dynamics - Bone: dyn_skirt_side_02_r!(SimulationSpace:AnimPhysSimSpaceType Enum, ComponentPose=@n10, Alpha=1, GravityOverride:Vector="0,0,-980", ExternalForce="0,0,0") -> (Pose) {
    angularDampingOverride=0.9999
    angularSpringConstant=400
    bAngularSpring=true
    bOverrideAngularDamping=true
    bUseGravityOverride=true
    gravityScale=0.35
    simulationSpace=RootRelative
    alphaBoolBlend {blendOption=Linear}
    boundBone {boneName=dyn_skirt_side_02_r}
    componentAppliedLinearAccClamp=<2000 2000 2000>
    componentLinearAccScale=<1 1 1>
    componentLinearVelScale=<1 1 1>
    gravityOverride {z=-980}
    physicsBodyDefinitions [{
      boundBone {boneName=dyn_skirt_side_02_r}
      boxExtents=<30 20 20>
      constraintSetup {angularLimitsMax=<0 0 0>, angularLimitsMin=<0 -25 -5>, angularTarget=<1 0 0>}
      localJointOffset=<3 0 0>
     }]
   }
   n12 = Anim Dynamics - Bone: dyn_skirt_front_02_01_r!(SimulationSpace:AnimPhysSimSpaceType Enum, ComponentPose=@n11, Alpha=1, GravityOverride:Vector="0,0,-980", ExternalForce="0,0,0") -> (Pose) {
    angularDampingOverride=0.9999
    angularSpringConstant=400
    bAngularSpring=true
    bOverrideAngularDamping=true
    bUseGravityOverride=true
    gravityScale=0.35
    simulationSpace=RootRelative
    alphaBoolBlend {blendOption=Linear}
    boundBone {boneName=dyn_skirt_front_02_01_r}
    componentAppliedLinearAccClamp=<2000 2000 2000>
    componentLinearAccScale=<1 1 1>
    componentLinearVelScale=<1 1 1>
    gravityOverride {z=-980}
    physicsBodyDefinitions [{
      boundBone {boneName=dyn_skirt_front_02_01_r}
      boxExtents=<30 20 20>
      constraintSetup {angularLimitsMax=<0 0 0>, angularLimitsMin=<0 -25 -5>, angularTarget=<1 0 0>}
      localJointOffset=<3 0 0>
     }]
   }
   n13 = Anim Dynamics - Bone: dyn_skirt_front_02_02_r!(SimulationSpace:AnimPhysSimSpaceType Enum, ComponentPose=@n12, Alpha=1, GravityOverride:Vector="0,0,-980", ExternalForce="0,0,0") -> (Pose) {
    angularDampingOverride=0.9999
    angularSpringConstant=400
    bAngularSpring=true
    bOverrideAngularDamping=true
    bUseGravityOverride=true
    gravityScale=0.35
    simulationSpace=RootRelative
    alphaBoolBlend {blendOption=Linear}
    boundBone {boneName=dyn_skirt_front_02_02_r}
    componentAppliedLinearAccClamp=<2000 2000 2000>
    componentLinearAccScale=<1 1 1>
    componentLinearVelScale=<1 1 1>
    gravityOverride {z=-980}
    physicsBodyDefinitions [{
      boundBone {boneName=dyn_skirt_front_02_02_r}
      boxExtents=<30 20 20>
      constraintSetup {angularLimitsMax=<0 0 0>, angularLimitsMin=<0 -25 -5>, angularTarget=<1 0 0>}
      localJointOffset=<3 0 0>
     }]
   }
   n14 = Anim Dynamics - Bone: dyn_skirt_front_02_03_r!(SimulationSpace:AnimPhysSimSpaceType Enum, ComponentPose=@n13, Alpha=1, GravityOverride:Vector="0,0,-980", ExternalForce="0,0,0") -> (Pose) {
    angularDampingOverride=0.9999
    angularSpringConstant=400
    bAngularSpring=true
    bOverrideAngularDamping=true
    bUseGravityOverride=true
    gravityScale=0.35
    simulationSpace=RootRelative
    alphaBoolBlend {blendOption=Linear}
    boundBone {boneName=dyn_skirt_front_02_03_r}
    componentAppliedLinearAccClamp=<2000 2000 2000>
    componentLinearAccScale=<1 1 1>
    componentLinearVelScale=<1 1 1>
    gravityOverride {z=-980}
    physicsBodyDefinitions [{
      boundBone {boneName=dyn_skirt_front_02_03_r}
      boxExtents=<30 20 20>
      constraintSetup {angularLimitsMax=<0 0 0>, angularLimitsMin=<0 -25 -5>, angularTarget=<1 0 0>}
      localJointOffset=<3 0 0>
     }]
   }
   n15 = Anim Dynamics - Bone: dyn_skirt_front_01_l!(SimulationSpace:AnimPhysSimSpaceType Enum, ComponentPose=@n14, Alpha=1, GravityOverride:Vector="0,0,-980", ExternalForce="0,0,0") -> (Pose) {
    angularDampingOverride=0.9999
    angularSpringConstant=400
    bAngularSpring=true
    bOverrideAngularDamping=true
    bUseGravityOverride=true
    gravityScale=0.35
    simulationSpace=RootRelative
    alphaBoolBlend {blendOption=Linear}
    boundBone {boneName=dyn_skirt_front_01_l}
    componentAppliedLinearAccClamp=<2000 2000 2000>
    componentLinearAccScale=<1 1 1>
    componentLinearVelScale=<1 1 1>
    gravityOverride {z=-980}
    physicsBodyDefinitions [{
      boundBone {boneName=dyn_skirt_front_01_l}
      boxExtents=<30 20 20>
      constraintSetup {angularLimitsMax=<0 0 0>, angularLimitsMin=<0 -25 -5>, angularTarget=<1 0 0>}
      localJointOffset=<3 0 0>
     }]
   }
   n16 = Anim Dynamics - Bone: base_skirt_front_02_r!(SimulationSpace:AnimPhysSimSpaceType Enum, ComponentPose=@n15, Alpha=1, GravityOverride:Vector="0,0,-980", ExternalForce="0,0,0") -> (Pose) {
    angularDampingOverride=0.9999
    angularSpringConstant=400
    bAngularSpring=true
    bOverrideAngularDamping=true
    bUseGravityOverride=true
    gravityScale=0.35
    simulationSpace=RootRelative
    alphaBoolBlend {blendOption=Linear}
    boundBone {boneName=base_skirt_front_02_r}
    componentAppliedLinearAccClamp=<2000 2000 2000>
    componentLinearAccScale=<1 1 1>
    componentLinearVelScale=<1 1 1>
    gravityOverride {z=-980}
    physicsBodyDefinitions [{
      boundBone {boneName=base_skirt_front_02_r}
      boxExtents=<30 20 20>
      constraintSetup {angularLimitsMax=<0 0 0>, angularLimitsMin=<0 -25 -5>, angularTarget=<1 0 0>}
      localJointOffset=<3 0 0>
     }]
   }
   n17 = "Transform (Modify) Bone - Bone: base_skirt_front_02_01_l"!(ComponentPose=@n16, bAlphaBoolEnabled:bool=True, Translation="1,0,2", Rotation="0,0,0", Scale="1,1,1") -> (Pose) {
    alphaInputType=Bool
    rotationMode=BMM_Additive
    rotationSpace=BCS_BoneSpace
    translationMode=BMM_Additive
    translationSpace=BCS_BoneSpace
    boneToModify {boneName=base_skirt_front_02_01_l}
    translation {x=1, z=2}
   }
   n18 = Anim Dynamics - Bone: dyn_necklace_03(ComponentPose=@n17, Alpha=1, ExternalForce="0,0,0") {
    angularDampingOverride=1
    angularSpringConstant=8000
    bAngularSpring=true
    bOverrideAngularDamping=true
    bUseGravityOverride=true
    simulationSpace=RootRelative
    alphaBoolBlend {blendOption=Linear}
    boundBone {boneName=dyn_necklace_03}
    componentAppliedLinearAccClamp=<4000 4000 4000>
    componentLinearAccScale=<0.2 0.2 0.2>
    componentLinearVelScale=<0.2 0.2 0.2>
    gravityOverride {z=-980}
    physicsBodyDefinitions [{
      boundBone {boneName=dyn_necklace_03}
      boxExtents=<20 15 15>
      constraintSetup {angularLimitsMax=<0 0 8>, angularLimitsMin=<0 -10 -8>, angularTarget=<1 0 0>}
      localJointOffset=<4 0 0>
     }]
   }
   n19 = Anim Dynamics - Bone: dyn_necklace_01(ComponentPose=@n18, Alpha=1, ExternalForce="0,0,0") {
    angularDampingOverride=1
    angularSpringConstant=4000
    bAngularSpring=true
    bOverrideAngularDamping=true
    bUseGravityOverride=true
    simulationSpace=RootRelative
    alphaBoolBlend {blendOption=Linear}
    boundBone {boneName=dyn_necklace_01}
    componentAppliedLinearAccClamp=<4000 4000 4000>
    componentLinearAccScale=<0.2 0.2 0.2>
    componentLinearVelScale=<0.2 0.2 0.2>
    gravityOverride {z=-980}
    physicsBodyDefinitions [{
      boundBone {boneName=dyn_necklace_01}
      boxExtents=<20 15 15>
      constraintSetup {angularLimitsMax=<0 0 5>, angularLimitsMin=<0 0 -5>, angularTarget=<1 0 0>}
      localJointOffset=<2 0 0>
     }]
   }
   n20 = Anim Dynamics - Bone: dyn_bangle_l(ComponentPose=@n19, Alpha=1, ExternalForce="0,0,0") {
    angularDampingOverride=1
    angularSpringConstant=20000
    bAngularSpring=true
    bOverrideAngularDamping=true
    bUseGravityOverride=true
    gravityScale=0.35
    simulationSpace=RootRelative
    alphaBoolBlend {blendOption=Linear}
    boundBone {boneName=dyn_bangle_l}
    componentAppliedLinearAccClamp=<2000 2000 2000>
    componentLinearAccScale=<0.5 0.5 0.5>
    componentLinearVelScale=<0.5 0.5 0.5>
    gravityOverride {z=-980}
    physicsBodyDefinitions [{
      boundBone {boneName=dyn_bangle_l}
      boxExtents=<10 15 15>
      constraintSetup {angularLimitsMax=<5 5 0>, angularLimitsMin=<-5 -5 0>, angularTarget=<1 0 0>}
      localJointOffset=<4 0 0>
     }]
   }
   n21 = Anim Dynamics - Bone: dyn_bangre_01_r(ComponentPose=@n20, Alpha=1, ExternalForce="0,0,0") {
    angularDampingOverride=1
    angularSpringConstant=20000
    bAngularSpring=true
    bOverrideAngularDamping=true
    bUseGravityOverride=true
    gravityScale=0.35
    simulationSpace=RootRelative
    alphaBoolBlend {blendOption=Linear}
    boundBone {boneName=dyn_bangre_01_r}
    componentAppliedLinearAccClamp=<2000 2000 2000>
    componentLinearAccScale=<0.5 0.5 0.5>
    componentLinearVelScale=<0.5 0.5 0.5>
    gravityOverride {z=-980}
    physicsBodyDefinitions [{
      boundBone {boneName=dyn_bangre_01_r}
      boxExtents=<10 15 15>
      constraintSetup {angularLimitsMax=<5 5 0>, angularLimitsMin=<-5 -5 0>, angularTarget=<1 0 0>}
      localJointOffset=<4 0 0>
     }]
   }
   n22 = "Transform (Modify) Bone - Bone: dyn_skirt_side_01_l"(ComponentPose=@n21, Alpha=1, Translation="0,0,0", Rotation="0,0,20", Scale="1,1,1") {
    rotationMode=BMM_Additive
    translationMode=BMM_Additive
    alphaBoolBlend {blendOption=Linear}
    alphaScaleBiasClamp {bInterpResult=true, interpSpeedIncreasing=6}
    boneToModify {boneName=dyn_skirt_side_01_l}
    rotation {roll=20}
   }
   n23 = "Transform (Modify) Bone - Bone: dyn_skirt_back_01"(ComponentPose=@n22, Alpha=1, Translation="0,0,0", Rotation="0,0,20", Scale="1,1,1") {
    rotationMode=BMM_Additive
    translationMode=BMM_Additive
    alphaBoolBlend {blendOption=Linear}
    alphaScaleBiasClamp {bInterpResult=true, interpSpeedIncreasing=6}
    boneToModify {boneName=dyn_skirt_back_01}
    rotation {roll=20}
   }
   n24 = "Transform (Modify) Bone - Bone: dyn_skirt_side_01_r"(ComponentPose=@n23, Alpha=1, Translation="0,0,0", Rotation="0,0,20", Scale="1,1,1") {
    rotationMode=BMM_Additive
    translationMode=BMM_Additive
    alphaBoolBlend {blendOption=Linear}
    alphaScaleBiasClamp {bInterpResult=true, interpSpeedIncreasing=6}
    boneToModify {boneName=dyn_skirt_side_01_r}
    rotation {roll=20}
   }
   n25 = "Transform (Modify) Bone - Bone: dyn_skirt_side_02_l"(ComponentPose=@n24, Alpha=1, Translation="0,0,0", Rotation="0,0,20", Scale="1,1,1") {
    rotationMode=BMM_Additive
    translationMode=BMM_Additive
    alphaBoolBlend {blendOption=Linear}
    alphaScaleBiasClamp {bInterpResult=true, interpSpeedIncreasing=6}
    boneToModify {boneName=dyn_skirt_side_02_l}
    rotation {roll=20}
   }
   n26 = "Transform (Modify) Bone - Bone: dyn_skirt_back_02"(ComponentPose=@n25, Alpha=1, Translation="0,0,0", Rotation="0,0,20", Scale="1,1,1") {
    rotationMode=BMM_Additive
    translationMode=BMM_Additive
    alphaBoolBlend {blendOption=Linear}
    alphaScaleBiasClamp {bInterpResult=true, interpSpeedIncreasing=6}
    boneToModify {boneName=dyn_skirt_back_02}
    rotation {roll=20}
   }
   n27 = "Transform (Modify) Bone - Bone: dyn_skirt_side_02_r"(ComponentPose=@n26, Alpha=1, Translation="0,0,0", Rotation="0,0,20", Scale="1,1,1") {
    rotationMode=BMM_Additive
    translationMode=BMM_Additive
    alphaBoolBlend {blendOption=Linear}
    alphaScaleBiasClamp {bInterpResult=true, interpSpeedIncreasing=6}
    boneToModify {boneName=dyn_skirt_side_02_r}
    rotation {roll=20}
   }
   n28 = Get bIsInFrontEnd()
   n29 = RigidBody!(ComponentPose=@n27, bAlphaBoolEnabled:bool=@n28, OverrideWorldGravity="0,0,0", ExternalForce) -> (Pose) {
    alphaInputType=Bool
    bClampLinearTranslationLimitToRefPose=true
    bOverrideWorldGravity=true
    overridePhysicsAsset=&/F_LRG_HipTripper_Frontend_Physics
    simulationSpace=BaseBoneSpace
    alphaBoolBlend {blendOption=Linear}
    alphaScaleBiasClamp {interpSpeedDecreasing=15, interpSpeedIncreasing=20}
    baseBoneRef {boneName=pelvis}
    overrideWorldGravity {z=-980}
    simSpaceSettings {maxAngularAcceleration=500, maxAngularVelocity=4, maxLinearAcceleration=700, velocityScaleZ=0, worldAlpha=0.5}
   }
   n30 = "Transform (Modify) Bone - Bone: dyn_skirt_front_02_03_l"!(ComponentPose=@n29, bAlphaBoolEnabled:bool=@n28, Translation="-2,0,0", Rotation="0,0,0", Scale="1,1,1") -> (Pose) {
    alphaInputType=Bool
    rotationMode=BMM_Additive
    translationMode=BMM_Additive
    alphaBoolBlend {blendOption=Linear}
    boneToModify {boneName=dyn_skirt_front_02_03_l}
    translation {x=-2}
   }
   n31 = "Transform (Modify) Bone - Bone: dyn_skirt_front_02_01_r"!(ComponentPose=@n30, bAlphaBoolEnabled:bool=@n28, Translation="0,0,0", Rotation="10,0,0", Scale="1,1,1") -> (Pose) {
    alphaInputType=Bool
    rotationMode=BMM_Additive
    translationMode=BMM_Additive
    alphaBoolBlend {blendOption=Linear}
    boneToModify {boneName=dyn_skirt_front_02_01_r}
    rotation {pitch=10}
   }
   n32 = "Transform (Modify) Bone - Bone: deform_upperarm_up_base_r"!(ComponentPose=@n31, bAlphaBoolEnabled:bool=@n28, Translation="0,0,-5", Rotation="0,0,0", Scale="1,1,1") -> (Pose) {
    alphaInputType=Bool
    rotationMode=BMM_Additive
    translationMode=BMM_Additive
    alphaBoolBlend {blendOption=Linear}
    boneToModify {boneName=deform_upperarm_up_base_r}
    translation {z=-5}
   }
   n33 = "Transform (Modify) Bone - Bone: dyn_necklace_02"!(ComponentPose=@n32, bAlphaBoolEnabled:bool=@n28, Translation="0,0,0", Rotation="0,0,-7", Scale="1,1,1") -> (Pose) {
    alphaInputType=Bool
    rotationMode=BMM_Additive
    translationMode=BMM_Additive
    alphaBoolBlend {blendOption=Linear}
    boneToModify {boneName=dyn_necklace_02}
    rotation {roll=-7}
   }
   n34 = "Transform (Modify) Bone - Bone: base_skirt_front_02_01_l"(ComponentPose=@n33, bAlphaBoolEnabled=@n28, Translation="0,2,0", Rotation="0,0,0", Scale="1,1,1") {
    alphaInputType=Bool
    rotationMode=BMM_Additive
    translationMode=BMM_Additive
    alphaBoolBlend {blendOption=Linear}
    boneToModify {boneName=base_skirt_front_02_01_l}
    translation {y=2}
   }
   Output Pose(Result=Component To Local(ComponentPose=@n34)) {layerGroup=None}
  ]}
 EventGraph {nodes [
   n1 = Event BlueprintUpdateAnimation()
   > Set Delta Time X(Delta Time X:double=@n1.DeltaTimeX) -> (Output_Get:double)
   > n2 = Sequence!(execute) -> (then_0, then_1, then_2:Exec, then_3:Exec)
   > n3 = "Gravity Override State Function (Parameter Asset) (Nativized)"(
    CharacterPartAnimInstance=Self-Reference()
    GameTimeInSeconds=GetGameTimeInSeconds()
    Parameters=&/F_LRG_HipTripper_Paramater
   )
   > Set Gravity Override(Gravity Override=@n3)
   Set simSpace(execute=@n2.then_1, simSpace:AnimPhysSimSpaceType Enum) -> (Output_Get:AnimPhysSimSpaceType Enum)
   > Branch(Condition=OR Boolean(A=Get bIsDBNO(), B=Get bIsInFrontEnd()))
   > n4 = Branch(Condition=Get bIsDBNOCarried())
   > Set simSpace(simSpace=World)
   Set simSpace(execute=@n4.else)
   n5 = GetOwnerSkeletalMeshComponent()
   n6 = BreakRotator(InRot=BreakTransform(InTransform=GetDeltaTransformFromRefPose(self=@n5, BoneName=thigh_l, BaseName=pelvis)).Rotation)
   n7 = Get bIsFalling()
   n8 = Get bIsJumping()
   Set jump(
    execute=@n2.then_3
    jump:double=float * float(A=MapRangeClamped(Value=@n6.Yaw, InRangeA=0, InRangeB=-45), B:bool=OR Boolean(A=@n8, B=@n7)) {is=title K2Node_PromotableOperator}
   ) -> (Output_Get:double)
   n9 = PlayerMovement(execute=@n2.then_2, AnimInstance=Self-Reference(), XDotProductCoefficient=100, YDotProductCoefficient=100, ZDotProductCoefficient=100)
   > Branch(Condition=Get bIsCrouching())
   n10 = BreakVector(InVec=MakeVector(X=@n9, Y=@n9.YMovement))
   > Set CrouchLeft(CrouchLeft:double="To Float (Boolean)"(InBool="float < float"(A=@n10, B=-50))) -> (Output_Get:double)
   > Set CrouchRight(CrouchRight:double="To Float (Boolean)"(InBool="float > float"(A=@n10, B=50))) -> (Output_Get:double)
   > Set CrouchForward(CrouchForward:double="To Float (Boolean)"(InBool="float > float"(A=@n10.Y, B=40))) -> (Output_Get:double)
   > Set CrouchBackward(CrouchBackward:double="To Float (Boolean)"(InBool="float < float"(A=@n10.Y, B=-40))) -> (Output_Get:double)
  ]}
]
variables [
 CrouchBackward {type=number}
 CrouchForward {type=number}
 CrouchLeft {type=number}
 CrouchRight {type=number}
 Delta Time X {type=number}
 Gravi {type=boolean}
 Gravity Override {type=Vector}
 Gravity Override Params {type=GravityOverrideParamsStruct}
 jump {type=number}
 simSpace {type=AnimPhysSimSpaceType}
]
