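# Physics asset of collision shapes for the large female HipTripper player body, named as the colliders for its cape.
# Capsules cover the head, clavicles, spine, pelvis and feet, with paired capsules on the middle spine bones and a zero-length, effectively spherical one on the head.
# Dynamic skirt bones at back and sides also carry capsules.
# Arm, thigh, calf and neck bodies are empty, and no constraints join the bodies.
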PhysicsAsset /BRCosmetics/Characters/Player/Female/Large/Bodies/F_LRG_HipTripper/Meshes/F_LRG_HipTripper_CapeColliders
bodies {
 calf_l []
 calf_r []
 clavicle_l [{shapeType=Capsule, length=6.5623, radius=7.5194, center=<3.9566 -1.0862 -2.1127>, rotation=<rot -77.5566 177.3915 7.7314>}]
 clavicle_r [{shapeType=Capsule, length=6.5623, radius=7.5194, center=<-3.9566 1.0862 2.1127>, rotation=<rot -77.5566 177.3915 -172.2686>}]
 dyn_skirt_back_01 [{shapeType=Capsule, length=15, radius=8, shapeName=dyn_skirt_back_01_capsule, center=<5.2363 0.1628 1.9368>, rotation=<rot 0 0 -90>}]
 dyn_skirt_back_02 [{shapeType=Capsule, length=15, radius=8, shapeName=dyn_skirt_back_02_capsule, center=<5 0 0>, rotation=<rot 0 0 -90>}]
 dyn_skirt_side_01_l [{shapeType=Capsule, length=15, radius=8, shapeName=dyn_skirt_side_01_l_capsule, center=<5.1008 -0.871 1.0963>, rotation=<rot 0 0 -90>}]
 dyn_skirt_side_01_r [{shapeType=Capsule, length=15, radius=8, shapeName=dyn_skirt_side_01_r_capsule, center=<5.0758 0.3611 0.8129>, rotation=<rot 0 0 -90>}]
 dyn_skirt_side_02_l [{shapeType=Capsule, length=15, radius=8, shapeName=dyn_skirt_side_02_l_capsule, center=<5 0 0>, rotation=<rot 0 0 -90>}]
 dyn_skirt_side_02_r [{shapeType=Capsule, length=15, radius=8, shapeName=dyn_skirt_side_02_r_capsule, center=<5 0 0>, rotation=<rot 0 0 -90>}]
 foot_l [{shapeType=Capsule, length=11.0765, radius=8.0344, center=<-4.1811 6.0027 -0.338>, rotation=<rot 0.0388 1.4701 -89.3982>}]
 foot_r [{shapeType=Capsule, length=11.0765, radius=8.0344, center=<4.1811 -6.0027 0.338>, rotation=<rot 0.0388 1.4701 90.6018>}]
 head [{shapeType=Capsule, length=0, radius=18, center=<9.0697 0.0746 0>, rotation=<rot 0 -90 0>}]
 lowerarm_l []
 lowerarm_r []
 neck_02 []
 pelvis [{shapeType=Capsule, length=4.6467, radius=16.1016, center=<-2.4471 -2.5289 -0.1981>}]
 spine_01 [{shapeType=Capsule, length=10.3888, radius=19.3638, center=<-8.8624 -8.9457 0>}]
 spine_02 [
  {shapeType=Capsule, length=12.1821, radius=15.0495, center=<-2.5662 -8.0408 0>}
  {shapeType=Capsule, length=20.6875, radius=9.3125, center=<-1.8317 1.9362 0>, rotation=<rot 90 0 0>}
 ]
 spine_03 [
  {shapeType=Capsule, length=4.4656, radius=14.2061, center=<1.3247 -3.7757 0>}
  {shapeType=Capsule, length=20.7188, radius=9.2812, center=<-1.2163 0.549 0>, rotation=<rot 90 180 -175.0732>}
 ]
 spine_04 [
  {shapeType=Capsule, length=7.5076, radius=13.8916, center=<5.6234 -2.5528 0>}
  {shapeType=Capsule, length=10.0312, radius=9.2812, center=<-1.2326 1.308 0>, rotation=<rot 90 0 0>}
 ]
 spine_05 [{shapeType=Capsule, length=8.178, radius=15.8571, center=<0.6969 2.0354 0>}]
 thigh_l []
 thigh_r []
 upperarm_l []
 upperarm_r []
}
constraints []
