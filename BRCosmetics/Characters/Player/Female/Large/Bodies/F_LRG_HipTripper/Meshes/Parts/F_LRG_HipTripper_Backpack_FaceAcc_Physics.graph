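# Physics asset for the braids of the large female HipTripper character's backpack/face accessory mesh, covering left and right braid chains and the torso.
# Each side chains six dynamic capsule bodies to a tiny base body, which is rotationally locked to the neck.
# Pelvis, spine and neck capsules form a rotationally locked torso chain.
# Only those locked joints override angular motion; braid joints store narrow swing angles, with twist values differing between sides.
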PhysicsAsset /BRCosmetics/Characters/Player/Female/Large/Bodies/F_LRG_HipTripper/Meshes/Parts/F_LRG_HipTripper_Backpack_FaceAcc_Physics
bodies {
 Base_braid_MSTR_r [{shapeType=Capsule, length=0.505, radius=0.505, shapeName=Base_braid_MSTR_r_capsule}]
 base_braid_MSTR_l [{shapeType=Capsule, length=0.505, radius=0.505, shapeName=base_braid_MSTR_l_capsule, rotation=<rot 0 -90 0>}]
 dyn_braid_01_l [{shapeType=Capsule, length=5, radius=3, shapeName=dyn_braid_01_l_capsule, center=<5 0 0.5>, rotation=<rot 90 0 0>}]
 dyn_braid_01_r [{shapeType=Capsule, length=5, radius=3, shapeName=dyn_braid_01_r_capsule, center=<5 0 0.5>, rotation=<rot 90 0 0>}]
 dyn_braid_02_l [{shapeType=Capsule, length=5, radius=3, shapeName=dyn_braid_02_l_capsule, center=<5 0 0.5>, rotation=<rot 90 0 0>}]
 dyn_braid_02_r [{shapeType=Capsule, length=5, radius=3, shapeName=dyn_braid_02_r_capsule, center=<5 0 0.5>, rotation=<rot 90 0 0>}]
 dyn_braid_03_l [{shapeType=Capsule, length=5, radius=3, shapeName=dyn_braid_03_l_capsule, center=<5 0 0.5>, rotation=<rot 90 0 0>}]
 dyn_braid_03_r [{shapeType=Capsule, length=5, radius=3, shapeName=dyn_braid_03_r_capsule, center=<5 0 0.5>, rotation=<rot 90 0 0>}]
 dyn_braid_04_l [{shapeType=Capsule, length=5, radius=3, shapeName=dyn_braid_04_l_capsule, center=<5 0 0.5>, rotation=<rot 90 0 0>}]
 dyn_braid_04_r [{shapeType=Capsule, length=5, radius=3, shapeName=dyn_braid_04_r_capsule, center=<5 0 0.5>, rotation=<rot 90 0 0>}]
 dyn_braid_05_l [{shapeType=Capsule, length=5, radius=3, shapeName=dyn_braid_05_l_capsule, center=<5 0 0.5>, rotation=<rot 90 0 0>}]
 dyn_braid_05_r [{shapeType=Capsule, length=5, radius=3, shapeName=dyn_braid_05_r_capsule, center=<5 0 0.5>, rotation=<rot 90 0 0>}]
 dyn_braid_06_l [{shapeType=Capsule, length=8, radius=3, shapeName=dyn_braid_06_l_capsule, center=<7 0 0.5>, rotation=<rot 90 0 0>}]
 dyn_braid_06_r [{shapeType=Capsule, length=8, radius=3, shapeName=dyn_braid_06_r_capsule, center=<7 0 0.5>, rotation=<rot 90 0 0>}]
 neck_01 [{shapeType=Capsule, length=10, radius=5, shapeName=neck_01_capsule, rotation=<rot 0 -90 0>}]
 pelvis [{shapeType=Capsule, length=53.8012, radius=12.8437, shapeName=pelvis_capsule, center=<0 -4.0353 0>, rotation=<rot 0 -90 0>}]
 spine_03 [{shapeType=Capsule, length=48.8813, radius=10, shapeName=spine_03_capsule, rotation=<rot 0 -90 0>}]
 spine_05 [{shapeType=Capsule, length=26.146, radius=10, shapeName=spine_05_capsule, rotation=<rot 0 -90 0>}]
}
constraints [
 {
  bone1Name=Base_braid_MSTR_r
  bone2Name=neck_01
  swing1LimitDegrees=45
  swing1Motion=Locked
  swing2LimitDegrees=45
  swing2Motion=Locked
  twistLimitDegrees=45
  twistMotion=Locked
 }
 {bone1Name=dyn_braid_01_r, bone2Name=Base_braid_MSTR_r, swing1LimitDegrees=10, swing2LimitDegrees=10, twistLimitDegrees=5}
 {bone1Name=dyn_braid_02_r, bone2Name=dyn_braid_01_r, swing1LimitDegrees=10, swing2LimitDegrees=10, twistLimitDegrees=5}
 {bone1Name=dyn_braid_03_r, bone2Name=dyn_braid_02_r, swing1LimitDegrees=10, swing2LimitDegrees=10, twistLimitDegrees=5}
 {bone1Name=dyn_braid_04_r, bone2Name=dyn_braid_03_r, swing1LimitDegrees=10, swing2LimitDegrees=10, twistLimitDegrees=5}
 {bone1Name=dyn_braid_05_r, bone2Name=dyn_braid_04_r, swing1LimitDegrees=10, swing2LimitDegrees=10, twistLimitDegrees=5}
 {bone1Name=dyn_braid_06_r, bone2Name=dyn_braid_05_r, swing1LimitDegrees=10, swing2LimitDegrees=10, twistLimitDegrees=5}
 {bone1Name=dyn_braid_01_l, bone2Name=base_braid_MSTR_l, swing1LimitDegrees=10, swing2LimitDegrees=10, twistLimitDegrees=0}
 {bone1Name=dyn_braid_02_l, bone2Name=dyn_braid_01_l, swing1LimitDegrees=10, swing2LimitDegrees=10, twistLimitDegrees=0}
 {bone1Name=dyn_braid_03_l, bone2Name=dyn_braid_02_l, swing1LimitDegrees=10, swing2LimitDegrees=10, twistLimitDegrees=0}
 {bone1Name=dyn_braid_04_l, bone2Name=dyn_braid_03_l, swing1LimitDegrees=10, swing2LimitDegrees=10, twistLimitDegrees=0}
 {bone1Name=dyn_braid_05_l, bone2Name=dyn_braid_04_l, swing1LimitDegrees=10, swing2LimitDegrees=10, twistLimitDegrees=0}
 {bone1Name=dyn_braid_06_l, bone2Name=dyn_braid_05_l, swing1LimitDegrees=10, swing2LimitDegrees=10, twistLimitDegrees=0}
 {
  bone1Name=base_braid_MSTR_l
  bone2Name=neck_01
  swing1LimitDegrees=45
  swing1Motion=Locked
  swing2LimitDegrees=45
  swing2Motion=Locked
  twistLimitDegrees=45
  twistMotion=Locked
 }
 {
  bone1Name=spine_05
  bone2Name=spine_03
  swing1LimitDegrees=45
  swing1Motion=Locked
  swing2LimitDegrees=45
  swing2Motion=Locked
  twistLimitDegrees=45
  twistMotion=Locked
 }
 {
  bone1Name=spine_03
  bone2Name=pelvis
  swing1LimitDegrees=45
  swing1Motion=Locked
  swing2LimitDegrees=45
  swing2Motion=Locked
  twistLimitDegrees=45
  twistMotion=Locked
 }
 {
  bone1Name=neck_01
  bone2Name=spine_05
  swing1LimitDegrees=45
  swing1Motion=Locked
  swing2LimitDegrees=45
  swing2Motion=Locked
  twistLimitDegrees=45
  twistMotion=Locked
 }
]
