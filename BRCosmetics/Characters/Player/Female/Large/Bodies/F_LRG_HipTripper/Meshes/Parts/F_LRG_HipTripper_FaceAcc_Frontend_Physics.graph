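# Physics asset for the HipTripper large female character's frontend face accessory, simulating a pair of braids.
# Each side has a six-capsule chain hanging from a braid base locked to the neck, with limited swing and zero twist between segments.
# Capsules on the pelvis, spine, neck and left upper arm are joined by rotation-locked constraints, forming a body proxy around the braids.
# Swing limits are slightly tighter on the left chain than on the right.
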PhysicsAsset /BRCosmetics/Characters/Player/Female/Large/Bodies/F_LRG_HipTripper/Meshes/Parts/F_LRG_HipTripper_FaceAcc_Frontend_Physics
bodies {
 Base_braid_MSTR_r [{shapeType=Capsule, length=0.505, radius=0.505, shapeName=Base_braid_MSTR_r_capsule}]
 base_braid_MSTR_l [{shapeType=Capsule, length=0.505, radius=0.505, shapeName=base_braid_MSTR_l_capsule, rotation=<rot 0 -90 0>}]
 dyn_braid_01_l [{shapeType=Capsule, length=5, radius=3, shapeName=dyn_braid_01_l_capsule, center=<5 0 1>, rotation=<rot 90 0 0>}]
 dyn_braid_01_r [{shapeType=Capsule, length=5, radius=3, shapeName=dyn_braid_01_r_capsule, center=<5 0 1>, rotation=<rot -90 0 0>}]
 dyn_braid_02_l [{shapeType=Capsule, length=5, radius=4, shapeName=dyn_braid_02_l_capsule, center=<5 0 1>, rotation=<rot 90 0 0>}]
 dyn_braid_02_r [{shapeType=Capsule, length=5, radius=3, shapeName=dyn_braid_02_r_capsule, center=<5 0 1.3682>, rotation=<rot -90 0 0>}]
 dyn_braid_03_l [{shapeType=Capsule, length=5, radius=4, shapeName=dyn_braid_03_l_capsule, center=<5 0 1>, rotation=<rot 90 0 0>}]
 dyn_braid_03_r [{shapeType=Capsule, length=5, radius=4, shapeName=dyn_braid_03_r_capsule, center=<5 0 0.9731>, rotation=<rot -90 0 0>}]
 dyn_braid_04_l [{shapeType=Capsule, length=5, radius=4, shapeName=dyn_braid_04_l_capsule, center=<5 0 1.5>, rotation=<rot 90 0 0>}]
 dyn_braid_04_r [{shapeType=Capsule, length=5, radius=4, shapeName=dyn_braid_04_r_capsule, center=<5 0 1.3682>, rotation=<rot -90 0 0>}]
 dyn_braid_05_l [{shapeType=Capsule, length=5, radius=6, shapeName=dyn_braid_05_l_capsule, center=<5 0 1.5>, rotation=<rot 90 0 0>}]
 dyn_braid_05_r [{shapeType=Capsule, length=5, radius=6, shapeName=dyn_braid_05_r_capsule, center=<5 0 1>, rotation=<rot -90 0 0>}]
 dyn_braid_06_l [{shapeType=Capsule, length=8, radius=8, shapeName=dyn_braid_06_l_capsule, center=<7 0 2>, rotation=<rot 110 0 0>}]
 dyn_braid_06_r [{shapeType=Capsule, length=8, radius=7, shapeName=dyn_braid_06_r_capsule, center=<7 0 2.5>, rotation=<rot -70 0 0>}]
 neck_01 [{shapeType=Capsule, length=10, radius=5, shapeName=neck_01_capsule, rotation=<rot 0 -90 0>}]
 pelvis [{shapeType=Capsule, length=52.9703, radius=18, shapeName=pelvis_capsule, center=<0 -4.0353 0>, rotation=<rot 0 -90 0>}]
 spine_03 [{shapeType=Capsule, length=48.853, radius=11, shapeName=spine_03_capsule, center=<-1.5541 -1.6716 0>, rotation=<rot 0 -90 0>}]
 spine_05 [
  {shapeType=Capsule, length=29.9058, radius=9, shapeName=spine_05_capsule, center=<0 0.5155 0>, rotation=<rot 0 -90 0>}
  {shapeType=Capsule, length=43.9995, radius=9.2531, shapeName=spine_05_capsule, center=<-6.5339 -1.6871 0>}
 ]
 upperarm_l [{shapeType=Capsule, length=18, radius=6, shapeName=upperarm_l_capsule, center=<12 0 0>, rotation=<rot 0 -90 90>}]
}
constraints [
 {
  bone1Name=Base_braid_MSTR_r
  bone2Name=neck_01
  swing1LimitDegrees=45
  swing1Motion=Locked
  swing2LimitDegrees=45
  swing2Motion=Locked
  twistLimitDegrees=45
  twistMotion=Locked
 }
 {bone1Name=dyn_braid_01_r, bone2Name=Base_braid_MSTR_r, swing1LimitDegrees=25, swing2LimitDegrees=45, twistLimitDegrees=0}
 {bone1Name=dyn_braid_02_r, bone2Name=dyn_braid_01_r, swing1LimitDegrees=25, swing2LimitDegrees=45, twistLimitDegrees=0}
 {bone1Name=dyn_braid_03_r, bone2Name=dyn_braid_02_r, swing1LimitDegrees=25, swing2LimitDegrees=45, twistLimitDegrees=0}
 {bone1Name=dyn_braid_04_r, bone2Name=dyn_braid_03_r, swing1LimitDegrees=25, swing2LimitDegrees=45, twistLimitDegrees=0}
 {bone1Name=dyn_braid_05_r, bone2Name=dyn_braid_04_r, swing1LimitDegrees=25, swing2LimitDegrees=45, twistLimitDegrees=0}
 {bone1Name=dyn_braid_06_r, bone2Name=dyn_braid_05_r, swing1LimitDegrees=25, swing2LimitDegrees=45, twistLimitDegrees=0}
 {bone1Name=dyn_braid_01_l, bone2Name=base_braid_MSTR_l, swing1LimitDegrees=20, swing2LimitDegrees=45, twistLimitDegrees=0}
 {bone1Name=dyn_braid_02_l, bone2Name=dyn_braid_01_l, swing1LimitDegrees=20, swing2LimitDegrees=45, twistLimitDegrees=0}
 {bone1Name=dyn_braid_03_l, bone2Name=dyn_braid_02_l, swing1LimitDegrees=20, swing2LimitDegrees=45, twistLimitDegrees=0}
 {bone1Name=dyn_braid_04_l, bone2Name=dyn_braid_03_l, swing1LimitDegrees=20, swing2LimitDegrees=45, twistLimitDegrees=0}
 {bone1Name=dyn_braid_05_l, bone2Name=dyn_braid_04_l, swing1LimitDegrees=20, swing2LimitDegrees=45, twistLimitDegrees=0}
 {bone1Name=dyn_braid_06_l, bone2Name=dyn_braid_05_l, swing1LimitDegrees=20, swing2LimitDegrees=45, twistLimitDegrees=0}
 {
  bone1Name=base_braid_MSTR_l
  bone2Name=neck_01
  swing1LimitDegrees=45
  swing1Motion=Locked
  swing2LimitDegrees=45
  swing2Motion=Locked
  twistLimitDegrees=45
  twistMotion=Locked
 }
 {
  bone1Name=spine_05
  bone2Name=spine_03
  swing1LimitDegrees=45
  swing1Motion=Locked
  swing2LimitDegrees=45
  swing2Motion=Locked
  twistLimitDegrees=45
  twistMotion=Locked
 }
 {
  bone1Name=spine_03
  bone2Name=pelvis
  swing1LimitDegrees=45
  swing1Motion=Locked
  swing2LimitDegrees=45
  swing2Motion=Locked
  twistLimitDegrees=45
  twistMotion=Locked
 }
 {
  bone1Name=neck_01
  bone2Name=spine_05
  swing1LimitDegrees=45
  swing1Motion=Locked
  swing2LimitDegrees=45
  swing2Motion=Locked
  twistLimitDegrees=45
  twistMotion=Locked
 }
 {
  bone1Name=upperarm_l
  bone2Name=spine_05
  swing1LimitDegrees=45
  swing1Motion=Locked
  swing2LimitDegrees=45
  swing2Motion=Locked
  twistLimitDegrees=45
  twistMotion=Locked
 }
]
