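# Animation Blueprint for the Hip Tripper face accessory part that adds secondary motion to its braids.
# Copied parent and body poses get head/spine constraints and head-turn offsets, then cache as 'pre sim'.
# Bool blends select cape, backpack or default RigidBody branches, using separate physics assets and additive braid-bone corrections.
# The EventGraph updates backbling flags, head rotation, jump, simulation spaces, and a parameter-asset gravity override.
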AnimBlueprint /BRCosmetics/Characters/Player/Female/Large/Bodies/F_LRG_HipTripper/Meshes/Parts/F_LRG_HipTripper_FaceAcc_AnimBP
target_skeleton=&/F_LRG_HipTripper_FaceAcc_Skeleton
graphs [
 AnimGraph {nodes [
   Component To Local()
   n1 = "Transform (Modify) Bone - Bone: dyn_hat_b_base"(Alpha=Get Head_Rot_Bk() -> (Head_Rot_Bk:double), Translation="0,0,3", Rotation="0,0,30", Scale="1,1,1") {
    rotationMode=BMM_Additive
    translationMode=BMM_Additive
    alphaBoolBlend {blendOption=Linear}
    boneToModify {boneName=dyn_hat_b_base}
    rotation {roll=30}
    translation {z=3}
   }
   Component To Local(ComponentPose=@n1)
   n2 = "Transform (Modify) Bone - Bone: dyn_braid_01_r"(Alpha=1, Translation="0,0,0", Rotation="0,0,0", Scale="1,1,1") {
    rotationMode=BMM_Additive
    translationMode=BMM_Additive
    alphaBoolBlend {blendOption=Linear}
    alphaScaleBiasClamp {bInterpResult=true, interpSpeedIncreasing=4}
    boneToModify {boneName=dyn_braid_01_r}
   }
   n3 = "Transform (Modify) Bone - Bone: dyn_braid_02_r"(ComponentPose=@n2, Alpha=1, Translation="0,0,0", Rotation="0,0,0", Scale="1,1,1") {
    rotationMode=BMM_Additive
    translationMode=BMM_Additive
    alphaBoolBlend {blendOption=Linear}
    alphaScaleBiasClamp {bInterpResult=true, interpSpeedIncreasing=4}
    boneToModify {boneName=dyn_braid_02_r}
   }
   n4 = "Transform (Modify) Bone - Bone: dyn_braid_03_r"(ComponentPose=@n3, Alpha=1, Translation="0,0,0", Rotation="0,0,40", Scale="1,1,1") {
    rotationMode=BMM_Additive
    translationMode=BMM_Additive
    alphaBoolBlend {blendOption=Linear}
    alphaScaleBiasClamp {bInterpResult=true, interpSpeedIncreasing=4}
    boneToModify {boneName=dyn_braid_03_r}
    rotation {roll=40}
   }
   n5 = "Transform (Modify) Bone - Bone: dyn_braid_04_r"(ComponentPose=@n4, Alpha=1, Translation="0,0,0", Rotation="0,0,30", Scale="1,1,1") {
    rotationMode=BMM_Additive
    translationMode=BMM_Additive
    alphaBoolBlend {blendOption=Linear}
    alphaScaleBiasClamp {bInterpResult=true, interpSpeedIncreasing=4}
    boneToModify {boneName=dyn_braid_04_r}
    rotation {roll=30}
   }
   n6 = "Transform (Modify) Bone - Bone: dyn_braid_05_r"(ComponentPose=@n5, Alpha=1, Translation="0,0,0", Rotation="0,0,30", Scale="1,1,1") {
    rotationMode=BMM_Additive
    translationMode=BMM_Additive
    alphaBoolBlend {blendOption=Linear}
    alphaScaleBiasClamp {bInterpResult=true, interpSpeedIncreasing=4}
    boneToModify {boneName=dyn_braid_05_r}
    rotation {roll=30}
   }
   n7 = "Transform (Modify) Bone - Bone: dyn_braid_06_r"(ComponentPose=@n6, Alpha=1, Translation="0,0,0", Rotation="0,0,20", Scale="1,1,1") {
    rotationMode=BMM_Additive
    translationMode=BMM_Additive
    alphaBoolBlend {blendOption=Linear}
    alphaScaleBiasClamp {bInterpResult=true, interpSpeedIncreasing=4}
    boneToModify {boneName=dyn_braid_06_r}
    rotation {roll=20}
   }
   n8 = "Transform (Modify) Bone - Bone: dyn_braid_01_l"(ComponentPose=@n7, Alpha=1, Translation="0,0,0", Rotation="0,0,0", Scale="1,1,1") {
    rotationMode=BMM_Additive
    translationMode=BMM_Additive
    alphaBoolBlend {blendOption=Linear}
    alphaScaleBiasClamp {bInterpResult=true, interpSpeedIncreasing=2}
    boneToModify {boneName=dyn_braid_01_l}
   }
   n9 = "Transform (Modify) Bone - Bone: dyn_braid_02_l"(ComponentPose=@n8, Alpha=1, Translation="0,0,0", Rotation="0,0,0", Scale="1,1,1") {
    rotationMode=BMM_Additive
    translationMode=BMM_Additive
    alphaBoolBlend {blendOption=Linear}
    alphaScaleBiasClamp {bInterpResult=true, interpSpeedIncreasing=2}
    boneToModify {boneName=dyn_braid_02_l}
   }
   n10 = "Transform (Modify) Bone - Bone: dyn_braid_03_l"(ComponentPose=@n9, Alpha=1, Translation="0,0,0", Rotation="0,0,40", Scale="1,1,1") {
    rotationMode=BMM_Additive
    translationMode=BMM_Additive
    alphaBoolBlend {blendOption=Linear}
    alphaScaleBiasClamp {bInterpResult=true, interpSpeedIncreasing=2}
    boneToModify {boneName=dyn_braid_03_l}
    rotation {roll=40}
   }
   n11 = "Transform (Modify) Bone - Bone: dyn_braid_04_l"(ComponentPose=@n10, Alpha=1, Translation="0,0,0", Rotation="0,0,30", Scale="1,1,1") {
    rotationMode=BMM_Additive
    translationMode=BMM_Additive
    alphaBoolBlend {blendOption=Linear}
    alphaScaleBiasClamp {bInterpResult=true, interpSpeedIncreasing=2}
    boneToModify {boneName=dyn_braid_04_l}
    rotation {roll=30}
   }
   n12 = "Transform (Modify) Bone - Bone: dyn_braid_05_l"(ComponentPose=@n11, Alpha=1, Translation="0,0,0", Rotation="0,0,20", Scale="1,1,1") {
    rotationMode=BMM_Additive
    translationMode=BMM_Additive
    alphaBoolBlend {blendOption=Linear}
    alphaScaleBiasClamp {bInterpResult=true, interpSpeedIncreasing=2}
    boneToModify {boneName=dyn_braid_05_l}
    rotation {roll=20}
   }
   "Transform (Modify) Bone - Bone: dyn_braid_06_l"(ComponentPose=@n12, Alpha=1, Translation="0,0,0", Rotation="0,0,10", Scale="1,1,1") {
    rotationMode=BMM_Additive
    translationMode=BMM_Additive
    alphaBoolBlend {blendOption=Linear}
    alphaScaleBiasClamp {bInterpResult=true, interpSpeedIncreasing=2}
    boneToModify {boneName=dyn_braid_06_l}
    rotation {roll=10}
   }
   n13 = Get Gravity Override()
   n14 = RigidBody!(ComponentPose=Local To Component(LocalPose=Use cached pose 'pre sim'()), Alpha, OverrideWorldGravity:Vector=@n13, ExternalForce) -> (Pose) {
    bClampLinearTranslationLimitToRefPose=true
    bOverrideWorldGravity=true
    overridePhysicsAsset=&/F_LRG_HipTripper_FaceAcc_Physics
    simulationSpace=BaseBoneSpace
    alphaBoolBlend {blendOption=Linear}
    baseBoneRef {boneName=spine_05}
    overrideWorldGravity {z=-920}
    simSpaceSettings {maxAngularAcceleration=500, maxAngularVelocity=10, maxLinearAcceleration=4500, maxLinearVelocity=900, velocityScaleZ=0, worldAlpha=1}
   }
   n15 = RigidBody(ComponentPose=@n14, OverrideWorldGravity=@n13) {
    bClampLinearTranslationLimitToRefPose=true
    bOverrideWorldGravity=true
    overridePhysicsAsset=&/F_LRG_HipTripper_FaceAcc_Physics_Jump
    simulationSpace=BaseBoneSpace
    alphaBoolBlend {blendOption=Linear}
    alphaScaleBiasClamp {bInterpResult=true, interpSpeedIncreasing=5}
    baseBoneRef {boneName=root}
    overrideWorldGravity {z=-920}
    simSpaceSettings {maxAngularAcceleration=500, maxAngularVelocity=10, maxLinearAcceleration=4500, maxLinearVelocity=900, worldAlpha=0.5}
   }
   n16 = RigidBody(ComponentPose=@n15, OverrideWorldGravity=@n13) {
    bClampLinearTranslationLimitToRefPose=true
    bOverrideWorldGravity=true
    overridePhysicsAsset=&/F_LRG_HipTripper_FaceAcc_Frontend_Physics
    simulationSpace=BaseBoneSpace
    alphaBoolBlend {blendOption=Linear}
    baseBoneRef {boneName=spine_05}
    overrideWorldGravity {z=-920}
    simSpaceSettings {maxAngularAcceleration=500, maxAngularVelocity=10, maxLinearAcceleration=4500, maxLinearVelocity=900, velocityScaleZ=0, worldAlpha=1}
   }
   n17 = "Transform (Modify) Bone - Bone: dyn_braid_01_l"!(ComponentPose=@n16, bAlphaBoolEnabled:bool=True, Translation="0,0,0", Rotation="10,0,10", Scale="1,1,1") -> (Pose) {
    alphaInputType=Bool
    rotationMode=BMM_Additive
    translationMode=BMM_Additive
    alphaBoolBlend {blendOption=Linear}
    boneToModify {boneName=dyn_braid_01_l}
    rotation {pitch=10, roll=10}
   }
   n18 = "Transform (Modify) Bone - Bone: dyn_braid_02_l"!(ComponentPose=@n17, bAlphaBoolEnabled:bool=True, Translation="0,0,0", Rotation="10,0,-8", Scale="1,1,1") -> (Pose) {
    alphaInputType=Bool
    rotationMode=BMM_Additive
    translationMode=BMM_Additive
    alphaBoolBlend {blendOption=Linear}
    boneToModify {boneName=dyn_braid_02_l}
    rotation {pitch=10, roll=-8}
   }
   n19 = "Transform (Modify) Bone - Bone: dyn_braid_03_l"!(ComponentPose=@n18, bAlphaBoolEnabled:bool=True, Translation="0,0,0", Rotation="-5,0,0", Scale="1,1,1") -> (Pose) {
    alphaInputType=Bool
    rotationMode=BMM_Additive
    translationMode=BMM_Additive
    alphaBoolBlend {blendOption=Linear}
    boneToModify {boneName=dyn_braid_03_l}
    rotation {pitch=-5}
   }
   n20 = "Transform (Modify) Bone - Bone: dyn_braid_04_l"!(ComponentPose=@n19, bAlphaBoolEnabled:bool=True, Translation="0,0,0", Rotation="0,0,-4", Scale="1,1,1") -> (Pose) {
    alphaInputType=Bool
    rotationMode=BMM_Additive
    translationMode=BMM_Additive
    alphaBoolBlend {blendOption=Linear}
    boneToModify {boneName=dyn_braid_04_l}
    rotation {roll=-4}
   }
   n21 = "Transform (Modify) Bone - Bone: dyn_braid_05_l"!(ComponentPose=@n20, bAlphaBoolEnabled:bool=True, Translation="0,0,0", Rotation="-5,0,-25", Scale="1,1,1") -> (Pose) {
    alphaInputType=Bool
    rotationMode=BMM_Additive
    translationMode=BMM_Additive
    alphaBoolBlend {blendOption=Linear}
    boneToModify {boneName=dyn_braid_05_l}
    rotation {pitch=-5, roll=-25}
   }
   n22 = "Transform (Modify) Bone - Bone: dyn_braid_06_l"!(ComponentPose=@n21, bAlphaBoolEnabled:bool=True, Translation="0,0,0", Rotation="0,0,-10", Scale="1,1,1") -> (Pose) {
    alphaInputType=Bool
    rotationMode=BMM_Additive
    translationMode=BMM_Additive
    alphaBoolBlend {blendOption=Linear}
    boneToModify {boneName=dyn_braid_06_l}
    rotation {roll=-10}
   }
   n23 = "Transform (Modify) Bone - Bone: dyn_braid_01_r"!(ComponentPose=@n22, bAlphaBoolEnabled:bool=True, Translation="0,0,0", Rotation="0,0,10", Scale="1,1,1") -> (Pose) {
    alphaInputType=Bool
    rotationMode=BMM_Additive
    translationMode=BMM_Additive
    alphaBoolBlend {blendOption=Linear}
    boneToModify {boneName=dyn_braid_01_r}
    rotation {roll=10}
   }
   n24 = "Transform (Modify) Bone - Bone: dyn_braid_02_r"!(ComponentPose=@n23, bAlphaBoolEnabled:bool=True, Translation="0,0,0", Rotation="0,0,-30", Scale="1,1,1") -> (Pose) {
    alphaInputType=Bool
    rotationMode=BMM_Additive
    translationMode=BMM_Additive
    alphaBoolBlend {blendOption=Linear}
    boneToModify {boneName=dyn_braid_02_r}
    rotation {roll=-30}
   }
   n25 = "Transform (Modify) Bone - Bone: dyn_braid_03_r"!(ComponentPose=@n24, bAlphaBoolEnabled:bool=True, Translation="0,0,0", Rotation="-5,0,10", Scale="1,1,1") -> (Pose) {
    alphaInputType=Bool
    rotationMode=BMM_Additive
    translationMode=BMM_Additive
    alphaBoolBlend {blendOption=Linear}
    boneToModify {boneName=dyn_braid_03_r}
    rotation {pitch=-5, roll=10}
   }
   n26 = "Transform (Modify) Bone - Bone: dyn_braid_04_r"!(ComponentPose=@n25, bAlphaBoolEnabled:bool=True, Translation="0,0,0", Rotation="5,0,5", Scale="1,1,1") -> (Pose) {
    alphaInputType=Bool
    rotationMode=BMM_Additive
    translationMode=BMM_Additive
    alphaBoolBlend {blendOption=Linear}
    boneToModify {boneName=dyn_braid_04_r}
    rotation {pitch=5, roll=5}
   }
   n27 = "Transform (Modify) Bone - Bone: dyn_braid_05_r"!(ComponentPose=@n26, bAlphaBoolEnabled:bool=True, Translation="0,0,0", Rotation="20,0,-8", Scale="1,1,1") -> (Pose) {
    alphaInputType=Bool
    rotationMode=BMM_Additive
    translationMode=BMM_Additive
    alphaBoolBlend {blendOption=Linear}
    boneToModify {boneName=dyn_braid_05_r}
    rotation {pitch=20, roll=-8}
   }
   n28 = "Transform (Modify) Bone - Bone: dyn_braid_06_r"!(ComponentPose=@n27, bAlphaBoolEnabled:bool=True, Translation="0,0,0", Rotation="0,0,2", Scale="1,1,1") -> (Pose) {
    alphaInputType=Bool
    rotationMode=BMM_Additive
    translationMode=BMM_Additive
    alphaBoolBlend {blendOption=Linear}
    boneToModify {boneName=dyn_braid_06_r}
    rotation {roll=2}
   }
   n29 = Component To Local(ComponentPose=@n28)
   n30 = "Constraint - dyn_braid_02_l (P)"(ComponentPose=Local To Component(LocalPose=Use cached pose 'pre sim'()), Alpha=0.8, ConstraintWeights_0=1) {
    alphaBoolBlend {blendOption=Linear}
    boneToModify {boneName=dyn_braid_02_l}
    constraintSetup [{targetBone {boneName=spine_05}}]
    constraintWeights [1]
   }
   n31 = "Constraint - dyn_braid_02_r (P)"(ComponentPose=@n30, Alpha=0.8, ConstraintWeights_0=1) {
    alphaBoolBlend {blendOption=Linear}
    boneToModify {boneName=dyn_braid_02_r}
    constraintSetup [{targetBone {boneName=spine_05}}]
    constraintWeights [1]
   }
   n32 = "Constraint - dyn_braid_03_l (PP)"!(ComponentPose=@n31, Alpha=0.8, ConstraintWeights_0=0.5, ConstraintWeights_1:float=0.5) -> (Pose) {
    alphaBoolBlend {blendOption=Linear}
    boneToModify {boneName=dyn_braid_03_l}
    constraintSetup [{targetBone {boneName=spine_04}}, {targetBone {boneName=spine_03}}]
    constraintWeights [1,1]
   }
   n33 = "Constraint - dyn_braid_03_r (PP)"!(ComponentPose=@n32, Alpha=0.8, ConstraintWeights_0=0.5, ConstraintWeights_1:float=0.5) -> (Pose) {
    alphaBoolBlend {blendOption=Linear}
    boneToModify {boneName=dyn_braid_03_r}
    constraintSetup [{targetBone {boneName=spine_04}}, {targetBone {boneName=spine_03}}]
    constraintWeights [1,1]
   }
   n34 = "Constraint - dyn_braid_04_l (P)"(ComponentPose=@n33, Alpha=0.5, ConstraintWeights_0=1) {
    alphaBoolBlend {blendOption=Linear}
    boneToModify {boneName=dyn_braid_04_l}
    constraintSetup [{targetBone {boneName=spine_03}}]
    constraintWeights [1]
   }
   n35 = "Constraint - dyn_braid_04_l (P)"(ComponentPose=@n34, Alpha=0.5, ConstraintWeights_0=1) {
    alphaBoolBlend {blendOption=Linear}
    boneToModify {boneName=dyn_braid_04_l}
    constraintSetup [{targetBone {boneName=spine_03}}]
    constraintWeights [1]
   }
   n36 = RigidBody(
    ComponentPose="Transform (Modify) Bone - Bone: dyn_braid_02_r"!(
     ComponentPose="Transform (Modify) Bone - Bone: dyn_braid_02_l"!(ComponentPose=@n35, Alpha:float=1, Translation="0,0,0", Rotation="0,0,0", Scale="1,1,1") -> (Pose) {rotationMode=BMM_Additive, alphaBoolBlend {blendOption=Linear}, boneToModify {boneName=dyn_braid_02_l}}
     Alpha:float=1
     Translation="0,0,0"
     Rotation="0,0,0"
     Scale="1,1,1"
    ) -> (Pose) {rotationMode=BMM_Additive, alphaBoolBlend {blendOption=Linear}, boneToModify {boneName=dyn_braid_02_r}}
    OverrideWorldGravity="0,0,-920"
   ) {
    bClampLinearTranslationLimitToRefPose=true
    bOverrideWorldGravity=true
    overridePhysicsAsset=&/F_LRG_HipTripper_Backpack_FaceAcc_Physics
    simulationSpace=BaseBoneSpace
    alphaBoolBlend {blendOption=Linear}
    baseBoneRef {boneName=spine_05}
    overrideWorldGravity {z=-920}
    simSpaceSettings {maxAngularAcceleration=40, maxAngularVelocity=3, maxLinearAcceleration=40, maxLinearVelocity=50, velocityScaleZ=0, worldAlpha=1}
   }
   n37 = "Transform (Modify) Bone - Bone: dyn_braid_04_r"!(ComponentPose=@n36, Alpha:float=1, Translation="0,0,0", Rotation="0,0,20", Scale="1,1,1") -> (Pose) {
    rotationMode=BMM_Additive
    translationMode=BMM_Additive
    alphaBoolBlend {blendOption=Linear}
    alphaScaleBiasClamp {bInterpResult=true, interpSpeedDecreasing=8, interpSpeedIncreasing=6}
    boneToModify {boneName=dyn_braid_04_r}
    rotation {roll=20}
   }
   n38 = "Transform (Modify) Bone - Bone: dyn_braid_05_r"!(ComponentPose=@n37, Alpha:float=1, Translation="0,0,0", Rotation="0,0,10", Scale="1,1,1") -> (Pose) {
    rotationMode=BMM_Additive
    translationMode=BMM_Additive
    alphaBoolBlend {blendOption=Linear}
    alphaScaleBiasClamp {bInterpResult=true, interpSpeedDecreasing=8, interpSpeedIncreasing=6}
    boneToModify {boneName=dyn_braid_05_r}
    rotation {roll=10}
   }
   n39 = "Transform (Modify) Bone - Bone: dyn_braid_06_r"!(ComponentPose=@n38, Alpha:float=1, Translation="0,0,0", Rotation="0,0,5", Scale="1,1,1") -> (Pose) {
    rotationMode=BMM_Additive
    translationMode=BMM_Additive
    alphaBoolBlend {blendOption=Linear}
    alphaScaleBiasClamp {bInterpResult=true, interpSpeedDecreasing=8, interpSpeedIncreasing=6}
    boneToModify {boneName=dyn_braid_06_r}
    rotation {roll=5}
   }
   n40 = "Transform (Modify) Bone - Bone: dyn_braid_04_l"!(ComponentPose=@n39, Alpha:float=1, Translation="0,0,0", Rotation="0,0,20", Scale="1,1,1") -> (Pose) {
    rotationMode=BMM_Additive
    translationMode=BMM_Additive
    alphaBoolBlend {blendOption=Linear}
    alphaScaleBiasClamp {bInterpResult=true, interpSpeedDecreasing=8, interpSpeedIncreasing=6}
    boneToModify {boneName=dyn_braid_04_l}
    rotation {roll=20}
   }
   n41 = "Transform (Modify) Bone - Bone: dyn_braid_05_l"!(ComponentPose=@n40, Alpha:float=1, Translation="0,0,0", Rotation="0,0,10", Scale="1,1,1") -> (Pose) {
    rotationMode=BMM_Additive
    translationMode=BMM_Additive
    alphaBoolBlend {blendOption=Linear}
    alphaScaleBiasClamp {bInterpResult=true, interpSpeedDecreasing=8, interpSpeedIncreasing=6}
    boneToModify {boneName=dyn_braid_05_l}
    rotation {roll=10}
   }
   n42 = "Transform (Modify) Bone - Bone: dyn_braid_06_l"!(ComponentPose=@n41, Alpha:float=1, Translation="0,0,0", Rotation="0,0,5", Scale="1,1,1") -> (Pose) {
    rotationMode=BMM_Additive
    translationMode=BMM_Additive
    alphaBoolBlend {blendOption=Linear}
    alphaScaleBiasClamp {bInterpResult=true, interpSpeedDecreasing=8, interpSpeedIncreasing=6}
    boneToModify {boneName=dyn_braid_06_l}
    rotation {roll=5}
   }
   n43 = Component To Local(ComponentPose=@n42)
   n44 = Get IsWearingBackbling() -> (IsWearingBackbling:bool)
   n45 = "Constraint - dyn_braid_02_l (P)"(ComponentPose=Local To Component(LocalPose=Use cached pose 'pre sim'()), Alpha=1, ConstraintWeights_0=1) {
    alphaBoolBlend {blendOption=Linear}
    boneToModify {boneName=dyn_braid_02_l}
    constraintSetup [{targetBone {boneName=spine_05}}]
    constraintWeights [1]
   }
   n46 = "Constraint - dyn_braid_02_r (P)"(ComponentPose=@n45, Alpha=1, ConstraintWeights_0=1) {
    alphaBoolBlend {blendOption=Linear}
    boneToModify {boneName=dyn_braid_02_r}
    constraintSetup [{targetBone {boneName=spine_05}}]
    constraintWeights [1]
   }
   n47 = "Constraint - dyn_braid_03_l (PP)"(ComponentPose=@n46, Alpha=0.8, ConstraintWeights_0=0.5, ConstraintWeights_1=0.5) {
    alphaBoolBlend {blendOption=Linear}
    boneToModify {boneName=dyn_braid_03_l}
    constraintSetup [{targetBone {boneName=spine_04}}, {targetBone {boneName=spine_03}}]
    constraintWeights [1,1]
   }
   n48 = "Constraint - dyn_braid_03_r (PP)"(ComponentPose=@n47, Alpha=0.8, ConstraintWeights_0=0.5, ConstraintWeights_1=0.5) {
    alphaBoolBlend {blendOption=Linear}
    boneToModify {boneName=dyn_braid_03_r}
    constraintSetup [{targetBone {boneName=spine_04}}, {targetBone {boneName=spine_03}}]
    constraintWeights [1,1]
   }
   n49 = "Constraint - dyn_braid_04_l (P)"(ComponentPose=@n48, Alpha=0.5, ConstraintWeights_0=1) {
    alphaBoolBlend {blendOption=Linear}
    boneToModify {boneName=dyn_braid_04_l}
    constraintSetup [{targetBone {boneName=spine_03}}]
    constraintWeights [1]
   }
   n50 = "Constraint - dyn_braid_04_l (P)"(ComponentPose=@n49, Alpha=0.5, ConstraintWeights_0=1) {
    alphaBoolBlend {blendOption=Linear}
    boneToModify {boneName=dyn_braid_04_l}
    constraintSetup [{targetBone {boneName=spine_03}}]
    constraintWeights [1]
   }
   n51 = RigidBody(
    ComponentPose="Transform (Modify) Bone - Bone: dyn_braid_02_r"(
     ComponentPose="Transform (Modify) Bone - Bone: dyn_braid_02_l"(ComponentPose=@n50, Alpha=1, Translation="0,0,0", Rotation="0,0,0", Scale="1,1,1") {rotationMode=BMM_Additive, alphaBoolBlend {blendOption=Linear}, boneToModify {boneName=dyn_braid_02_l}}
     Alpha=1
     Translation="0,0,0"
     Rotation="0,0,0"
     Scale="1,1,1"
    ) {rotationMode=BMM_Additive, alphaBoolBlend {blendOption=Linear}, boneToModify {boneName=dyn_braid_02_r}}
    OverrideWorldGravity="0,0,-920"
   ) {
    bClampLinearTranslationLimitToRefPose=true
    bOverrideWorldGravity=true
    overridePhysicsAsset=&/F_LRG_HipTripper_FaceAcc_Cape_Physics
    simulationSpace=BaseBoneSpace
    alphaBoolBlend {blendOption=Linear}
    baseBoneRef {boneName=main_braid_MSTR}
    overrideWorldGravity {z=-920}
    simSpaceSettings {maxAngularAcceleration=40, maxAngularVelocity=3, maxLinearAcceleration=40, maxLinearVelocity=50, velocityScaleZ=0, worldAlpha=1}
   }
   Output Pose(Result=Blend Poses by bool(
     BlendPose_0=Component To Local(ComponentPose=@n51)
     BlendPose_1=Blend Poses by bool(BlendPose_0=@n43, BlendPose_1=@n29, bActiveValue=@n44) {blendType=Linear}
     bActiveValue=Get Is Backbling A Cape() -> (Is Backbling A Cape:bool)
    ) {blendType=Linear}) {layerGroup=None}
   n52 = Layered blend per bone(BasePose=Copy Pose From Mesh() {bUseAttachedParent=true}, BlendPoses_0=Copy Pose From Mesh(SourceMeshComponent=Get Body_Mesh())) {layerSetup [{branchFilters [
       {boneName=upperarm_l}
       {boneName=upperarm_r}
       {boneName=lowerarm_l}
       {boneName=lowerarm_r}
       {boneName=thigh_l}
       {boneName=thigh_r}
       {boneName=calf_l}
       {boneName=calf_r}
      ]}]}
   n53 = "Constraint - main_braid_MSTR (PP)"!(ComponentPose=Local To Component(LocalPose=@n52), Alpha=1, ConstraintWeights_0=0.4, ConstraintWeights_1:float=0.6) -> (Pose) {
    alphaBoolBlend {blendOption=Linear}
    boneToModify {boneName=main_braid_MSTR}
    constraintSetup [{targetBone {boneName=spine_05}}, {targetBone {boneName=head}}]
    constraintWeights [1,1]
   }
   n54 = "Constraint - Base_braid_MSTR_r (PP)"!(ComponentPose=@n53, Alpha=1, ConstraintWeights_0=0.6, ConstraintWeights_1:float=0.4) -> (Pose) {
    alphaBoolBlend {blendOption=Linear}
    boneToModify {boneName=Base_braid_MSTR_r}
    constraintSetup [{targetBone {boneName=spine_05}}, {targetBone {boneName=head}}]
    constraintWeights [1,1]
   }
   n55 = "Constraint - base_braid_MSTR_l (PP)"!(ComponentPose=@n54, Alpha=1, ConstraintWeights_0=0.6, ConstraintWeights_1:float=0.4) -> (Pose) {
    alphaBoolBlend {blendOption=Linear}
    boneToModify {boneName=base_braid_MSTR_l}
    constraintSetup [{targetBone {boneName=spine_05}}, {targetBone {boneName=head}}]
    constraintWeights [1,1]
   }
   n56 = Get Head_Rot_Bk()
   n57 = "Transform (Modify) Bone - Bone: main_braid_MSTR"(ComponentPose=@n55, Alpha=@n56, Translation="0,-5,1", Rotation="0,0,20", Scale="1,1,1") {
    rotationMode=BMM_Additive
    translationMode=BMM_Additive
    alphaBoolBlend {blendOption=Linear}
    boneToModify {boneName=main_braid_MSTR}
    rotation {roll=20}
    translation {y=-5, z=1}
   }
   n58 = "Transform (Modify) Bone - Bone: Base_braid_MSTR_r"(ComponentPose=@n57, Alpha=@n56, Translation="0,0,0", Rotation="0,0,0", Scale="1,1,1") {rotationMode=BMM_Additive, translationMode=BMM_Additive, alphaBoolBlend {blendOption=Linear}, boneToModify {boneName=Base_braid_MSTR_r}}
   n59 = "Transform (Modify) Bone - Bone: base_braid_MSTR_l"(ComponentPose=@n58, Alpha=@n56, Translation="0,0,0", Rotation="0,0,10", Scale="1,1,1") {
    rotationMode=BMM_Additive
    translationMode=BMM_Additive
    alphaBoolBlend {blendOption=Linear}
    boneToModify {boneName=base_braid_MSTR_l}
    rotation {roll=10}
   }
   Save cached pose 'pre sim'(Pose=Component To Local(ComponentPose=@n59)) {cacheName=pre sim}
  ]}
 EventGraph {nodes [
   n1 = Event BlueprintUpdateAnimation()
   n2 = Sequence!(execute=[@n1.then, @n19.then]) -> (then_0, then_1, then_2:Exec, then_3:Exec, then_4:Exec, then_5:Exec)
   > Set Body_Mesh(Body_Mesh=GetOwnerSkeletalMeshComponent())
   n3 = TryGetPawnOwner()
   > Is Valid(InputObject=@n3)
   > n4 = Cast To FortPlayerPawnAthena(Object=@n3)
   > Set As Fort Player Pawn Athena(As Fort Player Pawn Athena:Fort Player Pawn Athena Object Reference=@n4) -> (Output_Get:Fort Player Pawn Athena Object Reference)
   > n5 = Cast To PlayerPawn_Athena(Object=@n3)
   > Set Player Pawn(Player Pawn:Player Pawn Athena Object Reference=@n5) -> (Output_Get:Player Pawn Athena Object Reference)
   n6 = GetOwnerSkeletalMeshComponent()
   n7 = BreakRotator(InRot=BreakTransform(InTransform=GetDeltaTransformFromRefPose(self=@n6, BoneName=thigh_l, BaseName=pelvis)).Rotation)
   n8 = Get bIsFalling()
   n9 = Get bIsJumping()
   Set jump(
    execute=@n2.then_2
    jump:double=float * float(A=MapRangeClamped(Value=@n7.Yaw, InRangeA=0, InRangeB=-45), B:bool=OR Boolean(A=@n9, B=@n8)) {is=title K2Node_PromotableOperator}
   ) -> (Output_Get:double)
   n10 = Get bIsPlayingEmote()
   > Set Front End No Emote(Front End No Emote:bool=AND Boolean(A=Get bIsInFrontEnd(), B=NOT Boolean(A=@n10))) -> (Output_Get:bool)
   n11 = Branch(execute=[@n2.then_3, @n20.then_1], Condition=Get bIsInFrontEnd())
   > Set Sim Space A(Sim Space A:AnimPhysSimSpaceType Enum=World) -> (Output_Get:AnimPhysSimSpaceType Enum)
   > Set Sim Space B(Sim Space B:AnimPhysSimSpaceType Enum=World) -> (Output_Get:AnimPhysSimSpaceType Enum)
   > Set Sim Space C(Sim Space C:AnimPhysSimSpaceType Enum=World) -> (Output_Get:AnimPhysSimSpaceType Enum)
   > Set Sim Space RB(Sim Space RB:Sim Space Settings Structure=Get Sim Space FE() -> (Sim Space FE:Sim Space Settings Structure)) -> (Output_Get:Sim Space Settings Structure)
   Set Sim Space A(execute=@n11.else)
   > Set Sim Space B(Sim Space B=BoneRelative)
   > Set Sim Space C(Sim Space C=Component)
   > Set Sim Space RB(Sim Space RB=Get Sim Space BR() -> (Sim Space BR:Sim Space Settings Structure))
   n12 = "Gravity Override State Function (Parameter Asset) (Nativized)"(
    execute=@n2.then_4
    CharacterPartAnimInstance=Self-Reference()
    GameTimeInSeconds=GetGameTimeInSeconds()
    Parameters=&/F_LRG_HipTripper_FaceAcc_Paramater
   )
   > Set Gravity Override(Gravity Override=@n12)
   n13 = GetOwnerSkeletalMeshComponent()
   > Set Head_Rot_Bk(Head_Rot_Bk:double=MapRangeClamped(
     Value=GetDeltaTransformFromRefPose!(self=@n13, BoneName=head, BaseName=spine_05) -> (
      ReturnValue_Location:Vector
      ReturnValue_Rotation_Roll:float
      ReturnValue_Rotation_Pitch:float
      ReturnValue_Rotation_Yaw:float
      ReturnValue_Scale:Vector
     ).ReturnValue_Rotation_Yaw
     InRangeA=-10
     InRangeB=-55
    )) -> (Output_Get:double)
   Set simSpace(execute=[@n2.then_5, @n20.then_3], simSpace:AnimPhysSimSpaceType Enum) -> (Output_Get:AnimPhysSimSpaceType Enum)
   > Branch(Condition=OR Boolean(A=Get bIsDBNO(), B=Get bIsInFrontEnd()))
   > n14 = Branch(Condition=Get bIsDBNOCarried())
   > Set simSpace(simSpace=World)
   Set simSpace(execute=@n14.else)
   n15 = CheckForBackbling()
   n16 = Set IsWearingBackbling(execute=[@n2.then_1, @n20.then_0], IsWearingBackbling:bool=@n15) -> (Output_Get:bool)
   > Set Is Backbling A Cape(Is Backbling A Cape:bool=@n15.bIsBackblingACape) -> (Output_Get:bool)
   > Set is Wear a Cape(is Wear a Cape:bool=AND Boolean(A=@n15, B=@n15.bIsBackblingLong)) -> (Output_Get:bool)
   > n17 = Branch(Condition=@n16)
   > n18 = Branch(Condition=@n15.bIsBackblingLong)
   Set IsShortBackblingOnAlpha(execute=[@n17.else, @n18.then], IsShortBackblingOnAlpha:double=0) -> (Output_Get:double)
   Set IsShortBackblingOnAlpha(execute=@n18.else, IsShortBackblingOnAlpha=1)
   n19 = Event OnCharacterCustomizationLoaded()
   n20 = Sequence!(execute) -> (then_0, then_1, then_2, then_3)
  ]}
]
variables [
 As Fort Player Pawn Athena {type=FortPlayerPawnAthena}
 Body_Mesh {type=SkeletalMeshComponent}
 Front End No Emote {type=boolean}
 Gravity Override {type=Vector}
 Gravity Override Params {type=GravityOverrideParamsStruct}
 Head_Rot_Bk {type=number}
 Is Backbling A Cape {type=boolean}
 is Wear a Cape {type=boolean}
 IsShortBackblingOnAlpha {type=number}
 IsWearingBackbling {type=boolean}
 jump {type=number}
 Player Pawn {type=PlayerPawn_Athena_C}
 Sim Space A {type=AnimPhysSimSpaceType}
 Sim Space B {type=AnimPhysSimSpaceType}
 Sim Space BR {
  type=SimSpaceSettings
  default {
   dampingAlpha=1
   maxAngularAcceleration=10000
   maxAngularVelocity=10000
   maxLinearAcceleration=10000
   maxLinearVelocity=10000
   velocityScaleZ=1
   worldAlpha=0
   externalLinearDragV=<0 0 0>
  }
 }
 Sim Space C {type=AnimPhysSimSpaceType}
 Sim Space FE {
  type=SimSpaceSettings
  default {
   dampingAlpha=1
   maxAngularAcceleration=10000
   maxAngularVelocity=10000
   maxLinearAcceleration=10000
   maxLinearVelocity=10000
   velocityScaleZ=1
   worldAlpha=0
   externalLinearDragV=<0 0 0>
  }
 }
 Sim Space RB {
  type=SimSpaceSettings
  default {
   dampingAlpha=1
   maxAngularAcceleration=10000
   maxAngularVelocity=10000
   maxLinearAcceleration=10000
   maxLinearVelocity=10000
   velocityScaleZ=1
   worldAlpha=0
   externalLinearDragV=<0 0 0>
  }
 }
 simSpace {type=AnimPhysSimSpaceType}
]
